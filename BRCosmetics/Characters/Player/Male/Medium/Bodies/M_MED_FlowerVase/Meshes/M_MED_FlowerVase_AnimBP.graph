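# Animation Blueprint for the male medium FlowerVase cosmetic body, adding procedural secondary motion to its collar and jacket bones.
# Jacket branches start from a 'PreSim Cache' pose, apply additive bone-space offsets to the panel bones, then run per-joint Anim Dynamics chains.
# One branch adds stronger, partly asymmetric lift and rotation on the upper jacket joints.
# Simulations are root-relative angular springs with tight limits, overridden gravity and damped component motion.
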AnimBlueprint /BRCosmetics/Characters/Player/Male/Medium/Bodies/M_MED_FlowerVase/Meshes/M_MED_FlowerVase_AnimBP
target_skeleton=&/M_MED_FlowerVase_Skeleton
graphs [
 AnimGraph {nodes [
   n1 = Anim Dynamics - Bone: dyn_Collar_c_01_jnt!(ComponentPose, Alpha=1, GravityOverride:Vector="0,0,0", ExternalForce="0,0,0") -> (Pose) {
    angularDampingOverride=1
    angularSpringConstant=20
    bAngularSpring=true
    bOverrideAngularDamping=true
    bUseGravityOverride=true
    simulationSpace=RootRelative
    alphaBoolBlend {blendOption=Linear}
    boundBone {boneName=dyn_Collar_c_01_jnt}
    componentAppliedLinearAccClamp=<6000 6000 6000>
    componentLinearAccScale=<0.4 0.4 0.4>
    componentLinearVelScale=<0.4 0.4 0.4>
    physicsBodyDefinitions [{
      boundBone {boneName=dyn_Collar_c_01_jnt}
      boxExtents=<10 10 10>
      constraintSetup {angularLimitsMax=<2 5 2>, angularLimitsMin=<-2 -5 -2>, angularTarget=<1 0 0>}
      localJointOffset=<3 0 0>
     }]
   }
   n2 = Anim Dynamics - Bone: dyn_Collar_l_01_jnt!(ComponentPose=@n1, Alpha=1, GravityOverride:Vector="0,0,0", ExternalForce="0,0,0") -> (Pose) {
    angularDampingOverride=1
    angularSpringConstant=20
    bAngularSpring=true
    bOverrideAngularDamping=true
    bUseGravityOverride=true
    simulationSpace=RootRelative
    alphaBoolBlend {blendOption=Linear}
    boundBone {boneName=dyn_Collar_l_01_jnt}
    componentAppliedLinearAccClamp=<6000 6000 6000>
    componentLinearAccScale=<0.4 0.4 0.4>
    componentLinearVelScale=<0.4 0.4 0.4>
    physicsBodyDefinitions [{
      boundBone {boneName=dyn_Collar_l_01_jnt}
      boxExtents=<10 10 10>
      constraintSetup {angularLimitsMax=<2 5 2>, angularLimitsMin=<-2 -5 -2>, angularTarget=<1 0 0>}
      localJointOffset=<3 0 0>
     }]
   }
   Anim Dynamics - Bone: dyn_Collar_r_01_jnt!(ComponentPose=@n2, Alpha=1, GravityOverride:Vector="0,0,0", ExternalForce="0,0,0") -> (Pose) {
    angularDampingOverride=1
    angularSpringConstant=20
    bAngularSpring=true
    bOverrideAngularDamping=true
    bUseGravityOverride=true
    simulationSpace=RootRelative
    alphaBoolBlend {blendOption=Linear}
    boundBone {boneName=dyn_Collar_r_01_jnt}
    componentAppliedLinearAccClamp=<6000 6000 6000>
    componentLinearAccScale=<0.4 0.4 0.4>
    componentLinearVelScale=<0.4 0.4 0.4>
    physicsBodyDefinitions [{
      boundBone {boneName=dyn_Collar_r_01_jnt}
      boxExtents=<10 10 10>
      constraintSetup {angularLimitsMax=<2 5 2>, angularLimitsMin=<-2 -5 -2>, angularTarget=<1 0 0>}
      localJointOffset=<3 0 0>
     }]
   }
   n3 = "Transform (Modify) Bone - Bone: dyn_Jacket_bk_c_MSTR"(Alpha=1, Translation="0,0,1", Rotation="2,0,0", Scale="1,1,1") {
    rotationMode=BMM_Additive
    rotationSpace=BCS_BoneSpace
    translationMode=BMM_Additive
    translationSpace=BCS_BoneSpace
    alphaBoolBlend {blendOption=Linear}
    boneToModify {boneName=dyn_Jacket_bk_c_MSTR}
    rotation {pitch=2}
    translation {z=1}
   }
   n4 = "Transform (Modify) Bone - Bone: dyn_Jacket_bk_r_MSTR"(ComponentPose=@n3, Alpha=1, Translation="0,0,1", Rotation="2,0,0", Scale="1,1,1") {
    rotationMode=BMM_Additive
    rotationSpace=BCS_BoneSpace
    translationMode=BMM_Additive
    translationSpace=BCS_BoneSpace
    alphaBoolBlend {blendOption=Linear}
    boneToModify {boneName=dyn_Jacket_bk_r_MSTR}
    rotation {pitch=2}
    translation {z=1}
   }
   n5 = "Transform (Modify) Bone - Bone: dyn_Jacket_bk_l_MSTR"(ComponentPose=@n4, Alpha=1, Translation="0,0,1", Rotation="2,0,0", Scale="1,1,1") {
    rotationMode=BMM_Additive
    rotationSpace=BCS_BoneSpace
    translationMode=BMM_Additive
    translationSpace=BCS_BoneSpace
    alphaBoolBlend {blendOption=Linear}
    boneToModify {boneName=dyn_Jacket_bk_l_MSTR}
    rotation {pitch=2}
    translation {z=1}
   }
   n6 = "Transform (Modify) Bone - Bone: dyn_Jacket_l_MSTR"(ComponentPose=@n5, Alpha=1, Translation="0,0,2", Rotation="2,0,0", Scale="1,1,1") {
    rotationMode=BMM_Additive
    rotationSpace=BCS_BoneSpace
    translationMode=BMM_Additive
    translationSpace=BCS_BoneSpace
    alphaBoolBlend {blendOption=Linear}
    boneToModify {boneName=dyn_Jacket_l_MSTR}
    rotation {pitch=2}
    translation {z=2}
   }
   n7 = "Transform (Modify) Bone - Bone: dyn_Jacket_r_MSTR"(ComponentPose=@n6, Alpha=1, Translation="0,0,2", Rotation="2,0,0", Scale="1,1,1") {
    rotationMode=BMM_Additive
    rotationSpace=BCS_BoneSpace
    translationMode=BMM_Additive
    translationSpace=BCS_BoneSpace
    alphaBoolBlend {blendOption=Linear}
    boneToModify {boneName=dyn_Jacket_r_MSTR}
    rotation {pitch=2}
    translation {z=2}
   }
   n8 = "Transform (Modify) Bone - Bone: dyn_Jacket_ft_l_MSTR"(ComponentPose=@n7, Alpha=1, Translation="0,0,2", Rotation="2,0,0", Scale="1,1,1") {
    rotationMode=BMM_Additive
    rotationSpace=BCS_BoneSpace
    translationMode=BMM_Additive
    translationSpace=BCS_BoneSpace
    alphaBoolBlend {blendOption=Linear}
    boneToModify {boneName=dyn_Jacket_ft_l_MSTR}
    rotation {pitch=2}
    translation {z=2}
   }
   "Transform (Modify) Bone - Bone: dyn_Jacket_ft_r_MSTR"(ComponentPose=@n8, Alpha=1, Translation="0,0,2", Rotation="2,0,0", Scale="1,1,1") {
    rotationMode=BMM_Additive
    rotationSpace=BCS_BoneSpace
    translationMode=BMM_Additive
    translationSpace=BCS_BoneSpace
    alphaBoolBlend {blendOption=Linear}
    boneToModify {boneName=dyn_Jacket_ft_r_MSTR}
    rotation {pitch=2}
    translation {z=2}
   }
   "Transform (Modify) Bone - Bone: attach_cape"(Alpha=1, Translation="0,0,0", Rotation="0,0,10", Scale="1,1,1") {
    rotationMode=BMM_Additive
    rotationSpace=BCS_BoneSpace
    scaleSpace=BCS_BoneSpace
    translationMode=BMM_Additive
    translationSpace=BCS_BoneSpace
    alphaBoolBlend {blendOption=Linear}
    alphaScaleBiasClamp {interpSpeedDecreasing=4}
    boneToModify {boneName=attach_cape}
    rotation {roll=10}
   }
   n9 = "Transform (Modify) Bone - Bone: dyn_Jacket_bk_r_MSTR"(ComponentPose=Local To Component(LocalPose=Use cached pose 'PreSim Cache'()), Alpha=1, Translation="0,0,1", Rotation="0,0,0", Scale="1,1,1") {
    rotationMode=BMM_Additive
    rotationSpace=BCS_BoneSpace
    scaleMode=BMM_Replace
    scaleSpace=BCS_BoneSpace
    translationMode=BMM_Additive
    translationSpace=BCS_BoneSpace
    alphaBoolBlend {blendOption=Linear}
    alphaScaleBiasClamp {bInterpResult=true, interpSpeedDecreasing=4}
    boneToModify {boneName=dyn_Jacket_bk_r_MSTR}
    translation {z=1}
   }
   n10 = "Transform (Modify) Bone - Bone: dyn_Jacket_bk_l_MSTR"(ComponentPose=@n9, Alpha=1, Translation="0,0,1", Rotation="0,0,0", Scale="1,1,1") {
    rotationMode=BMM_Additive
    rotationSpace=BCS_BoneSpace
    scaleMode=BMM_Replace
    scaleSpace=BCS_BoneSpace
    translationMode=BMM_Additive
    translationSpace=BCS_BoneSpace
    alphaBoolBlend {blendOption=Linear}
    alphaScaleBiasClamp {bInterpResult=true, interpSpeedDecreasing=4}
    boneToModify {boneName=dyn_Jacket_bk_l_MSTR}
    translation {z=1}
   }
   n11 = "Transform (Modify) Bone - Bone: dyn_Jacket_bk_c_MSTR"(ComponentPose=@n10, Alpha=1, Translation="0,0,1", Rotation="0,0,0", Scale="1,1,1") {
    rotationMode=BMM_Additive
    rotationSpace=BCS_BoneSpace
    scaleMode=BMM_Replace
    scaleSpace=BCS_BoneSpace
    translationMode=BMM_Additive
    translationSpace=BCS_BoneSpace
    alphaBoolBlend {blendOption=Linear}
    alphaScaleBiasClamp {bInterpResult=true, interpSpeedDecreasing=4}
    boneToModify {boneName=dyn_Jacket_bk_c_MSTR}
    translation {z=1}
   }
   n12 = "Transform (Modify) Bone - Bone: dyn_Jacket_ft_l_MSTR"(ComponentPose=@n11, Alpha=1, Translation="0,0,1", Rotation="10,0,0", Scale="1,1,1") {
    rotationMode=BMM_Additive
    rotationSpace=BCS_BoneSpace
    scaleMode=BMM_Replace
    scaleSpace=BCS_BoneSpace
    translationMode=BMM_Additive
    translationSpace=BCS_BoneSpace
    alphaBoolBlend {blendOption=Linear}
    alphaScaleBiasClamp {interpSpeedDecreasing=4}
    boneToModify {boneName=dyn_Jacket_ft_l_MSTR}
    rotation {pitch=10}
    translation {z=1}
   }
   n13 = "Transform (Modify) Bone - Bone: dyn_Jacket_ft_r_MSTR"(ComponentPose=@n12, Alpha=1, Translation="0,0,1", Rotation="10,0,0", Scale="1,1,1") {
    rotationMode=BMM_Additive
    rotationSpace=BCS_BoneSpace
    scaleMode=BMM_Replace
    scaleSpace=BCS_BoneSpace
    translationMode=BMM_Additive
    translationSpace=BCS_BoneSpace
    alphaBoolBlend {blendOption=Linear}
    alphaScaleBiasClamp {interpSpeedDecreasing=4}
    boneToModify {boneName=dyn_Jacket_ft_r_MSTR}
    rotation {pitch=10}
    translation {z=1}
   }
   n14 = "Transform (Modify) Bone - Bone: dyn_Jacket_l_MSTR"(ComponentPose=@n13, Alpha=1, Translation="0,0,2", Rotation="20,0,0", Scale="1,1,1") {
    rotationMode=BMM_Additive
    rotationSpace=BCS_BoneSpace
    scaleMode=BMM_Replace
    scaleSpace=BCS_BoneSpace
    translationMode=BMM_Additive
    translationSpace=BCS_BoneSpace
    alphaBoolBlend {blendOption=Linear}
    alphaScaleBiasClamp {interpSpeedDecreasing=4}
    boneToModify {boneName=dyn_Jacket_l_MSTR}
    rotation {pitch=20}
    translation {z=2}
   }
   n15 = "Transform (Modify) Bone - Bone: dyn_Jacket_r_MSTR"(ComponentPose=@n14, Alpha=1, Translation="0,0,2", Rotation="20,0,0", Scale="1,1,1") {
    rotationMode=BMM_Additive
    rotationSpace=BCS_BoneSpace
    scaleMode=BMM_Replace
    scaleSpace=BCS_BoneSpace
    translationMode=BMM_Additive
    translationSpace=BCS_BoneSpace
    alphaBoolBlend {blendOption=Linear}
    alphaScaleBiasClamp {interpSpeedDecreasing=4}
    boneToModify {boneName=dyn_Jacket_r_MSTR}
    rotation {pitch=20}
    translation {z=2}
   }
   n16 = Anim Dynamics - Bone: dyn_Jacket_ft_l_01_jnt!(SimulationSpace:AnimPhysSimSpaceType Enum, ComponentPose=@n15, Alpha=1, GravityOverride:Vector="0,0,-980", ExternalForce="0,0,0") -> (Pose) {
    angularDampingOverride=1
    angularSpringConstant=50
    bAngularSpring=true
    bOverrideAngularDamping=true
    bUseGravityOverride=true
    gravityScale=0.35
    simulationSpace=RootRelative
    alphaBoolBlend {blendOption=Linear}
    boundBone {boneName=dyn_Jacket_ft_l_01_jnt}
    componentLinearAccScale=<0.4 0.4 0.4>
    componentLinearVelScale=<0.4 0.4 0.4>
    gravityOverride {z=-980}
    physicsBodyDefinitions [{
      boundBone {boneName=dyn_Jacket_ft_l_01_jnt}
      boxExtents=<14 14 14>
      constraintSetup {angularLimitsMax=<1 0 5>, angularLimitsMin=<-1 -15 -5>, angularTarget=<1 0 0>}
      localJointOffset=<2 0 0>
     }]
   }
   n17 = Anim Dynamics - Bone: dyn_Jacket_ft_l_02_jnt!(SimulationSpace:AnimPhysSimSpaceType Enum, ComponentPose=@n16, Alpha=1, GravityOverride:Vector="0,0,-1000", ExternalForce="0,0,0") -> (Pose) {
    angularDampingOverride=1
    angularSpringConstant=50
    bAngularSpring=true
    bOverrideAngularDamping=true
    bUseGravityOverride=true
    gravityScale=0.35
    simulationSpace=RootRelative
    alphaBoolBlend {blendOption=Linear}
    boundBone {boneName=dyn_Jacket_ft_l_02_jnt}
    componentLinearAccScale=<0.4 0.4 0.4>
    componentLinearVelScale=<0.4 0.4 0.4>
    gravityOverride {z=-1000}
    physicsBodyDefinitions [{
      boundBone {boneName=dyn_Jacket_ft_l_02_jnt}
      boxExtents=<14 14 14>
      constraintSetup {angularLimitsMax=<1 0 5>, angularLimitsMin=<-1 -15 -5>, angularTarget=<1 0 0>}
      localJointOffset=<2 0 0>
     }]
   }
   n18 = Anim Dynamics - Bone: dyn_Jacket_ft_l_03_jnt!(SimulationSpace:AnimPhysSimSpaceType Enum, ComponentPose=@n17, Alpha=1, GravityOverride:Vector="0,0,-1000", ExternalForce="0,0,0") -> (Pose) {
    angularDampingOverride=1
    angularSpringConstant=50
    bAngularSpring=true
    bOverrideAngularDamping=true
    bUseGravityOverride=true
    gravityScale=0.35
    simulationSpace=RootRelative
    alphaBoolBlend {blendOption=Linear}
    boundBone {boneName=dyn_Jacket_ft_l_03_jnt}
    componentLinearAccScale=<0.4 0.4 0.4>
    componentLinearVelScale=<0.4 0.4 0.4>
    gravityOverride {z=-1000}
    physicsBodyDefinitions [{
      boundBone {boneName=dyn_Jacket_ft_l_03_jnt}
      boxExtents=<14 14 14>
      constraintSetup {angularLimitsMax=<1 0 5>, angularLimitsMin=<-1 -15 -5>, angularTarget=<1 0 0>}
      localJointOffset=<2 0 0>
     }]
   }
   n19 = Anim Dynamics - Bone: dyn_Jacket_ft_l_04_jnt!(SimulationSpace:AnimPhysSimSpaceType Enum, ComponentPose=@n18, Alpha=1, GravityOverride:Vector="0,0,-1000", ExternalForce="0,0,0") -> (Pose) {
    angularDampingOverride=1
    angularSpringConstant=50
    bAngularSpring=true
    bOverrideAngularDamping=true
    bUseGravityOverride=true
    gravityScale=0.35
    simulationSpace=RootRelative
    alphaBoolBlend {blendOption=Linear}
    boundBone {boneName=dyn_Jacket_ft_l_04_jnt}
    componentLinearAccScale=<0.4 0.4 0.4>
    componentLinearVelScale=<0.4 0.4 0.4>
    gravityOverride {z=-1000}
    physicsBodyDefinitions [{
      boundBone {boneName=dyn_Jacket_ft_l_04_jnt}
      boxExtents=<14 14 14>
      constraintSetup {angularLimitsMax=<1 0 5>, angularLimitsMin=<-1 -15 -5>, angularTarget=<1 0 0>}
      localJointOffset=<2 0 0>
     }]
   }
   n20 = Anim Dynamics - Bone: dyn_Jacket_l_01_jnt!(SimulationSpace:AnimPhysSimSpaceType Enum, ComponentPose=@n19, Alpha=1, GravityOverride:Vector="0,0,-980", ExternalForce="0,0,0") -> (Pose) {
    angularDampingOverride=1
    angularSpringConstant=50
    bAngularSpring=true
    bOverrideAngularDamping=true
    bUseGravityOverride=true
    gravityScale=0.35
    simulationSpace=RootRelative
    alphaBoolBlend {blendOption=Linear}
    boundBone {boneName=dyn_Jacket_l_01_jnt}
    componentLinearAccScale=<0.4 0.4 0.4>
    componentLinearVelScale=<0.4 0.4 0.4>
    gravityOverride {z=-980}
    physicsBodyDefinitions [{
      boundBone {boneName=dyn_Jacket_l_01_jnt}
      boxExtents=<14 14 14>
      constraintSetup {angularLimitsMax=<1 0 5>, angularLimitsMin=<-1 -15 -5>, angularTarget=<1 0 0>}
      localJointOffset=<2 0 0>
     }]
   }
   n21 = Anim Dynamics - Bone: dyn_Jacket_l_02_jnt!(SimulationSpace:AnimPhysSimSpaceType Enum, ComponentPose=@n20, Alpha=1, GravityOverride:Vector="0,0,-1000", ExternalForce="0,0,0") -> (Pose) {
    angularDampingOverride=1
    angularSpringConstant=50
    bAngularSpring=true
    bOverrideAngularDamping=true
    bUseGravityOverride=true
    gravityScale=0.35
    simulationSpace=RootRelative
    alphaBoolBlend {blendOption=Linear}
    boundBone {boneName=dyn_Jacket_l_02_jnt}
    componentLinearAccScale=<0.4 0.4 0.4>
    componentLinearVelScale=<0.4 0.4 0.4>
    gravityOverride {z=-1000}
    physicsBodyDefinitions [{
      boundBone {boneName=dyn_Jacket_l_02_jnt}
      boxExtents=<14 14 14>
      constraintSetup {angularLimitsMax=<1 0 5>, angularLimitsMin=<-1 -15 -5>, angularTarget=<1 0 0>}
      localJointOffset=<2 0 0>
     }]
   }
   n22 = Anim Dynamics - Bone: dyn_Jacket_l_03_jnt!(SimulationSpace:AnimPhysSimSpaceType Enum, ComponentPose=@n21, Alpha=1, GravityOverride:Vector="0,0,-1000", ExternalForce="0,0,0") -> (Pose) {
    angularDampingOverride=1
    angularSpringConstant=50
    bAngularSpring=true
    bOverrideAngularDamping=true
    bUseGravityOverride=true
    gravityScale=0.35
    simulationSpace=RootRelative
    alphaBoolBlend {blendOption=Linear}
    boundBone {boneName=dyn_Jacket_l_03_jnt}
    componentLinearAccScale=<0.4 0.4 0.4>
    componentLinearVelScale=<0.4 0.4 0.4>
    gravityOverride {z=-1000}
    physicsBodyDefinitions [{
      boundBone {boneName=dyn_Jacket_l_03_jnt}
      boxExtents=<14 14 14>
      constraintSetup {angularLimitsMax=<1 0 5>, angularLimitsMin=<-1 -15 -5>, angularTarget=<1 0 0>}
      localJointOffset=<2 0 0>
     }]
   }
   n23 = Anim Dynamics - Bone: dyn_Jacket_l_04_jnt!(SimulationSpace:AnimPhysSimSpaceType Enum, ComponentPose=@n22, Alpha=1, GravityOverride:Vector="0,0,-1000", ExternalForce="0,0,0") -> (Pose) {
    angularDampingOverride=1
    angularSpringConstant=50
    bAngularSpring=true
    bOverrideAngularDamping=true
    bUseGravityOverride=true
    gravityScale=0.35
    simulationSpace=RootRelative
    alphaBoolBlend {blendOption=Linear}
    boundBone {boneName=dyn_Jacket_l_04_jnt}
    componentLinearAccScale=<0.4 0.4 0.4>
    componentLinearVelScale=<0.4 0.4 0.4>
    gravityOverride {z=-1000}
    physicsBodyDefinitions [{
      boundBone {boneName=dyn_Jacket_l_04_jnt}
      boxExtents=<14 14 14>
      constraintSetup {angularLimitsMax=<1 0 5>, angularLimitsMin=<-1 -15 -5>, angularTarget=<1 0 0>}
      localJointOffset=<2 0 0>
     }]
   }
   n24 = Anim Dynamics - Bone: dyn_Jacket_ft_r_01_jnt!(SimulationSpace:AnimPhysSimSpaceType Enum, ComponentPose=@n23, Alpha=1, GravityOverride:Vector="0,0,-980", ExternalForce="0,0,0") -> (Pose) {
    angularDampingOverride=1
    angularSpringConstant=50
    bAngularSpring=true
    bOverrideAngularDamping=true
    bUseGravityOverride=true
    gravityScale=0.35
    simulationSpace=RootRelative
    alphaBoolBlend {blendOption=Linear}
    boundBone {boneName=dyn_Jacket_ft_r_01_jnt}
    componentLinearAccScale=<0.4 0.4 0.4>
    componentLinearVelScale=<0.4 0.4 0.4>
    gravityOverride {z=-980}
    physicsBodyDefinitions [{
      boundBone {boneName=dyn_Jacket_ft_r_01_jnt}
      boxExtents=<14 14 14>
      constraintSetup {angularLimitsMax=<1 0 5>, angularLimitsMin=<-1 -15 -5>, angularTarget=<1 0 0>}
      localJointOffset=<2 0 0>
     }]
   }
   n25 = Anim Dynamics - Bone: dyn_Jacket_ft_r_02_jnt!(SimulationSpace:AnimPhysSimSpaceType Enum, ComponentPose=@n24, Alpha=1, GravityOverride:Vector="0,0,-1000", ExternalForce="0,0,0") -> (Pose) {
    angularDampingOverride=1
    angularSpringConstant=50
    bAngularSpring=true
    bOverrideAngularDamping=true
    bUseGravityOverride=true
    gravityScale=0.35
    simulationSpace=RootRelative
    alphaBoolBlend {blendOption=Linear}
    boundBone {boneName=dyn_Jacket_ft_r_02_jnt}
    componentLinearAccScale=<0.4 0.4 0.4>
    componentLinearVelScale=<0.4 0.4 0.4>
    gravityOverride {z=-1000}
    physicsBodyDefinitions [{
      boundBone {boneName=dyn_Jacket_ft_r_02_jnt}
      boxExtents=<14 14 14>
      constraintSetup {angularLimitsMax=<1 0 5>, angularLimitsMin=<-1 -15 -5>, angularTarget=<1 0 0>}
      localJointOffset=<2 0 0>
     }]
   }
   n26 = Anim Dynamics - Bone: dyn_Jacket_ft_r_03_jnt!(SimulationSpace:AnimPhysSimSpaceType Enum, ComponentPose=@n25, Alpha=1, GravityOverride:Vector="0,0,-1000", ExternalForce="0,0,0") -> (Pose) {
    angularDampingOverride=1
    angularSpringConstant=50
    bAngularSpring=true
    bOverrideAngularDamping=true
    bUseGravityOverride=true
    gravityScale=0.35
    simulationSpace=RootRelative
    alphaBoolBlend {blendOption=Linear}
    boundBone {boneName=dyn_Jacket_ft_r_03_jnt}
    componentLinearAccScale=<0.4 0.4 0.4>
    componentLinearVelScale=<0.4 0.4 0.4>
    gravityOverride {z=-1000}
    physicsBodyDefinitions [{
      boundBone {boneName=dyn_Jacket_ft_r_03_jnt}
      boxExtents=<14 14 14>
      constraintSetup {angularLimitsMax=<1 0 5>, angularLimitsMin=<-1 -15 -5>, angularTarget=<1 0 0>}
      localJointOffset=<2 0 0>
     }]
   }
   n27 = Anim Dynamics - Bone: dyn_Jacket_ft_r_04_jnt!(SimulationSpace:AnimPhysSimSpaceType Enum, ComponentPose=@n26, Alpha=1, GravityOverride:Vector="0,0,-1000", ExternalForce="0,0,0") -> (Pose) {
    angularDampingOverride=1
    angularSpringConstant=50
    bAngularSpring=true
    bOverrideAngularDamping=true
    bUseGravityOverride=true
    gravityScale=0.35
    simulationSpace=RootRelative
    alphaBoolBlend {blendOption=Linear}
    boundBone {boneName=dyn_Jacket_ft_r_04_jnt}
    componentLinearAccScale=<0.4 0.4 0.4>
    componentLinearVelScale=<0.4 0.4 0.4>
    gravityOverride {z=-1000}
    physicsBodyDefinitions [{
      boundBone {boneName=dyn_Jacket_ft_r_04_jnt}
      boxExtents=<14 14 14>
      constraintSetup {angularLimitsMax=<1 0 5>, angularLimitsMin=<-1 -15 -5>, angularTarget=<1 0 0>}
      localJointOffset=<2 0 0>
     }]
   }
   n28 = Anim Dynamics - Bone: dyn_Jacket_r_01_jnt!(SimulationSpace:AnimPhysSimSpaceType Enum, ComponentPose=@n27, Alpha=1, GravityOverride:Vector="0,0,-980", ExternalForce="0,0,0") -> (Pose) {
    angularDampingOverride=1
    angularSpringConstant=50
    bAngularSpring=true
    bOverrideAngularDamping=true
    bUseGravityOverride=true
    gravityScale=0.35
    simulationSpace=RootRelative
    alphaBoolBlend {blendOption=Linear}
    boundBone {boneName=dyn_Jacket_r_01_jnt}
    componentLinearAccScale=<0.4 0.4 0.4>
    componentLinearVelScale=<0.4 0.4 0.4>
    gravityOverride {z=-980}
    physicsBodyDefinitions [{
      boundBone {boneName=dyn_Jacket_r_01_jnt}
      boxExtents=<14 14 14>
      constraintSetup {angularLimitsMax=<1 0 5>, angularLimitsMin=<-1 -15 -5>, angularTarget=<1 0 0>}
      localJointOffset=<2 0 0>
     }]
   }
   n29 = Anim Dynamics - Bone: dyn_Jacket_r_02_jnt!(SimulationSpace:AnimPhysSimSpaceType Enum, ComponentPose=@n28, Alpha=1, GravityOverride:Vector="0,0,-1000", ExternalForce="0,0,0") -> (Pose) {
    angularDampingOverride=1
    angularSpringConstant=50
    bAngularSpring=true
    bOverrideAngularDamping=true
    bUseGravityOverride=true
    gravityScale=0.35
    simulationSpace=RootRelative
    alphaBoolBlend {blendOption=Linear}
    boundBone {boneName=dyn_Jacket_r_02_jnt}
    componentLinearAccScale=<0.4 0.4 0.4>
    componentLinearVelScale=<0.4 0.4 0.4>
    gravityOverride {z=-1000}
    physicsBodyDefinitions [{
      boundBone {boneName=dyn_Jacket_r_02_jnt}
      boxExtents=<14 14 14>
      constraintSetup {angularLimitsMax=<1 0 5>, angularLimitsMin=<-1 -15 -5>, angularTarget=<1 0 0>}
      localJointOffset=<2 0 0>
     }]
   }
   n30 = Anim Dynamics - Bone: dyn_Jacket_r_03_jnt!(SimulationSpace:AnimPhysSimSpaceType Enum, ComponentPose=@n29, Alpha=1, GravityOverride:Vector="0,0,-1000", ExternalForce="0,0,0") -> (Pose) {
    angularDampingOverride=1
    angularSpringConstant=50
    bAngularSpring=true
    bOverrideAngularDamping=true
    bUseGravityOverride=true
    gravityScale=0.35
    simulationSpace=RootRelative
    alphaBoolBlend {blendOption=Linear}
    boundBone {boneName=dyn_Jacket_r_03_jnt}
    componentLinearAccScale=<0.4 0.4 0.4>
    componentLinearVelScale=<0.4 0.4 0.4>
    gravityOverride {z=-1000}
    physicsBodyDefinitions [{
      boundBone {boneName=dyn_Jacket_r_03_jnt}
      boxExtents=<14 14 14>
      constraintSetup {angularLimitsMax=<1 0 5>, angularLimitsMin=<-1 -15 -5>, angularTarget=<1 0 0>}
      localJointOffset=<2 0 0>
     }]
   }
   n31 = Anim Dynamics - Bone: dyn_Jacket_r_04_jnt!(SimulationSpace:AnimPhysSimSpaceType Enum, ComponentPose=@n30, Alpha=1, GravityOverride:Vector="0,0,-1000", ExternalForce="0,0,0") -> (Pose) {
    angularDampingOverride=1
    angularSpringConstant=50
    bAngularSpring=true
    bOverrideAngularDamping=true
    bUseGravityOverride=true
    gravityScale=0.35
    simulationSpace=RootRelative
    alphaBoolBlend {blendOption=Linear}
    boundBone {boneName=dyn_Jacket_r_04_jnt}
    componentLinearAccScale=<0.4 0.4 0.4>
    componentLinearVelScale=<0.4 0.4 0.4>
    gravityOverride {z=-1000}
    physicsBodyDefinitions [{
      boundBone {boneName=dyn_Jacket_r_04_jnt}
      boxExtents=<14 14 14>
      constraintSetup {angularLimitsMax=<1 0 5>, angularLimitsMin=<-1 -15 -5>, angularTarget=<1 0 0>}
      localJointOffset=<2 0 0>
     }]
   }
   n32 = Component To Local(ComponentPose=@n31)
   n33 = "Transform (Modify) Bone - Bone: dyn_Jacket_bk_r_MSTR"(ComponentPose=Local To Component(LocalPose=Use cached pose 'PreSim Cache'()), Alpha=1, Translation="0,0,1", Rotation="0,0,0", Scale="1,1,1") {
    rotationMode=BMM_Additive
    rotationSpace=BCS_BoneSpace
    scaleMode=BMM_Replace
    scaleSpace=BCS_BoneSpace
    translationMode=BMM_Additive
    translationSpace=BCS_BoneSpace
    alphaBoolBlend {blendOption=Linear}
    alphaScaleBiasClamp {bInterpResult=true, interpSpeedDecreasing=4}
    boneToModify {boneName=dyn_Jacket_bk_r_MSTR}
    translation {z=1}
   }
   n34 = "Transform (Modify) Bone - Bone: dyn_Jacket_bk_l_MSTR"(ComponentPose=@n33, Alpha=1, Translation="0,0,1", Rotation="0,0,0", Scale="1,1,1") {
    rotationMode=BMM_Additive
    rotationSpace=BCS_BoneSpace
    scaleMode=BMM_Replace
    scaleSpace=BCS_BoneSpace
    translationMode=BMM_Additive
    translationSpace=BCS_BoneSpace
    alphaBoolBlend {blendOption=Linear}
    alphaScaleBiasClamp {bInterpResult=true, interpSpeedDecreasing=4}
    boneToModify {boneName=dyn_Jacket_bk_l_MSTR}
    translation {z=1}
   }
   n35 = "Transform (Modify) Bone - Bone: dyn_Jacket_bk_c_MSTR"(ComponentPose=@n34, Alpha=1, Translation="0,0,1", Rotation="0,0,0", Scale="1,1,1") {
    rotationMode=BMM_Additive
    rotationSpace=BCS_BoneSpace
    scaleMode=BMM_Replace
    scaleSpace=BCS_BoneSpace
    translationMode=BMM_Additive
    translationSpace=BCS_BoneSpace
    alphaBoolBlend {blendOption=Linear}
    alphaScaleBiasClamp {bInterpResult=true, interpSpeedDecreasing=4}
    boneToModify {boneName=dyn_Jacket_bk_c_MSTR}
    translation {z=1}
   }
   n36 = "Transform (Modify) Bone - Bone: dyn_Jacket_bk_c_MSTR"(ComponentPose=@n35, Alpha=1, Translation="0,0,0", Rotation="20,0,0", Scale="1,1,1") {
    rotationMode=BMM_Additive
    rotationSpace=BCS_BoneSpace
    translationMode=BMM_Additive
    translationSpace=BCS_BoneSpace
    alphaBoolBlend {blendOption=Linear}
    alphaScaleBiasClamp {bInterpResult=true, interpSpeedIncreasing=4}
    boneToModify {boneName=dyn_Jacket_bk_c_MSTR}
    rotation {pitch=20}
   }
   n37 = "Transform (Modify) Bone - Bone: dyn_Jacket_bk_r_MSTR"(ComponentPose=@n36, Alpha=1, Translation="0,0,0", Rotation="20,0,0", Scale="1,1,1") {
    rotationMode=BMM_Additive
    rotationSpace=BCS_BoneSpace
    translationMode=BMM_Additive
    translationSpace=BCS_BoneSpace
    alphaBoolBlend {blendOption=Linear}
    alphaScaleBiasClamp {bInterpResult=true, interpSpeedIncreasing=4}
    boneToModify {boneName=dyn_Jacket_bk_r_MSTR}
    rotation {pitch=20}
   }
   n38 = "Transform (Modify) Bone - Bone: dyn_Jacket_bk_l_MSTR"(ComponentPose=@n37, Alpha=1, Translation="0,0,0", Rotation="20,0,0", Scale="1,1,1") {
    rotationMode=BMM_Additive
    rotationSpace=BCS_BoneSpace
    translationMode=BMM_Additive
    translationSpace=BCS_BoneSpace
    alphaBoolBlend {blendOption=Linear}
    alphaScaleBiasClamp {bInterpResult=true, interpSpeedIncreasing=4}
    boneToModify {boneName=dyn_Jacket_bk_l_MSTR}
    rotation {pitch=20}
   }
   n39 = "Transform (Modify) Bone - Bone: dyn_Jacket_bk_l_01_jnt"(ComponentPose=@n38, Alpha=1, Translation="0,0,5", Rotation="50,0,0", Scale="1,1,1") {
    rotationMode=BMM_Additive
    rotationSpace=BCS_BoneSpace
    scaleMode=BMM_Replace
    scaleSpace=BCS_BoneSpace
    translationMode=BMM_Additive
    translationSpace=BCS_BoneSpace
    alphaBoolBlend {blendOption=Linear}
    alphaScaleBiasClamp {bInterpResult=true, interpSpeedIncreasing=4}
    boneToModify {boneName=dyn_Jacket_bk_l_01_jnt}
    rotation {pitch=50}
    translation {z=5}
   }
   n40 = "Transform (Modify) Bone - Bone: dyn_Jacket_l_01_jnt"(ComponentPose=@n39, Alpha=1, Translation="0,0,5", Rotation="45,0,0", Scale="1,1,1") {
    rotationMode=BMM_Additive
    rotationSpace=BCS_BoneSpace
    scaleMode=BMM_Replace
    scaleSpace=BCS_BoneSpace
    translationMode=BMM_Additive
    translationSpace=BCS_BoneSpace
    alphaBoolBlend {blendOption=Linear}
    alphaScaleBiasClamp {bInterpResult=true, interpSpeedIncreasing=4}
    boneToModify {boneName=dyn_Jacket_l_01_jnt}
    rotation {pitch=45}
    translation {z=5}
   }
   n41 = "Transform (Modify) Bone - Bone: dyn_Jacket_ft_l_01_jnt"(ComponentPose=@n40, Alpha=1, Translation="0,0,0", Rotation="20,-20,0", Scale="1,1,1") {
    rotationMode=BMM_Additive
    rotationSpace=BCS_BoneSpace
    scaleMode=BMM_Replace
    scaleSpace=BCS_BoneSpace
    translationMode=BMM_Additive
    translationSpace=BCS_BoneSpace
    alphaBoolBlend {blendOption=Linear}
    alphaScaleBiasClamp {bInterpResult=true, interpSpeedIncreasing=4}
    boneToModify {boneName=dyn_Jacket_ft_l_01_jnt}
    rotation {pitch=20, yaw=-20}
   }
   n42 = "Transform (Modify) Bone - Bone: dyn_Jacket_bk_c_01_jnt"(ComponentPose=@n41, Alpha=1, Translation="0,0,5", Rotation="50,0,0", Scale="1,1,1") {
    rotationMode=BMM_Additive
    rotationSpace=BCS_BoneSpace
    scaleMode=BMM_Replace
    scaleSpace=BCS_BoneSpace
    translationMode=BMM_Additive
    translationSpace=BCS_BoneSpace
    alphaBoolBlend {blendOption=Linear}
    alphaScaleBiasClamp {bInterpResult=true, interpSpeedIncreasing=4}
    boneToModify {boneName=dyn_Jacket_bk_c_01_jnt}
    rotation {pitch=50}
    translation {z=5}
   }
   n43 = "Transform (Modify) Bone - Bone: dyn_Jacket_bk_r_01_jnt"(ComponentPose=@n42, Alpha=1, Translation="0,0,5", Rotation="50,0,0", Scale="1,1,1") {
    rotationMode=BMM_Additive
    rotationSpace=BCS_BoneSpace
    scaleMode=BMM_Replace
    scaleSpace=BCS_BoneSpace
    translationMode=BMM_Additive
    translationSpace=BCS_BoneSpace
    alphaBoolBlend {blendOption=Linear}
    alphaScaleBiasClamp {bInterpResult=true, interpSpeedIncreasing=4}
    boneToModify {boneName=dyn_Jacket_bk_r_01_jnt}
    rotation {pitch=50}
    translation {z=5}
   }
   n44 = "Transform (Modify) Bone - Bone: dyn_Jacket_r_01_jnt"(ComponentPose=@n43, Alpha=1, Translation="0,0,5", Rotation="45,0,0", Scale="1,1,1") {
    rotationMode=BMM_Additive
    rotationSpace=BCS_BoneSpace
    scaleMode=BMM_Replace
    scaleSpace=BCS_BoneSpace
    translationMode=BMM_Additive
    translationSpace=BCS_BoneSpace
    alphaBoolBlend {blendOption=Linear}
    alphaScaleBiasClamp {bInterpResult=true, interpSpeedIncreasing=4}
    boneToModify {boneName=dyn_Jacket_r_01_jnt}
    rotation {pitch=45}
    translation {z=5}
   }
   n45 = "Transform (Modify) Bone - Bone: dyn_Jacket_ft_r_01_jnt"(ComponentPose=@n44, Alpha=1, Translation="0,0,0", Rotation="20,20,0", Scale="1,1,1") {
    rotationMode=BMM_Additive
    rotationSpace=BCS_BoneSpace
    scaleMode=BMM_Replace
    scaleSpace=BCS_BoneSpace
    translationMode=BMM_Additive
    translationSpace=BCS_BoneSpace
    alphaBoolBlend {blendOption=Linear}
    alphaScaleBiasClamp {bInterpResult=true, interpSpeedIncreasing=4}
    boneToModify {boneName=dyn_Jacket_ft_r_01_jnt}
    rotation {pitch=20, yaw=20}
   }
   n46 = "Transform (Modify) Bone - Bone: dyn_Jacket_bk_l_01_jnt"(ComponentPose=@n45, Alpha=1, Translation="0,0,2", Rotation="30,0,0", Scale="1,1,1") {
    rotationMode=BMM_Additive
    rotationSpace=BCS_BoneSpace
    scaleMode=BMM_Replace
    scaleSpace=BCS_BoneSpace
    translationMode=BMM_Additive
    translationSpace=BCS_BoneSpace
    alphaBoolBlend {blendOption=Linear}
    alphaScaleBiasClamp {bInterpResult=true, interpSpeedIncreasing=4}
    boneToModify {boneName=dyn_Jacket_bk_l_01_jnt}
    rotation {pitch=30}
    translation {z=2}
   }
   n47 = "Transform (Modify) Bone - Bone: dyn_Jacket_l_01_jnt"(ComponentPose=@n46, Alpha=1, Translation="0,0,2", Rotation="15,0,0", Scale="1,1,1") {
    rotationMode=BMM_Additive
    rotationSpace=BCS_BoneSpace
    scaleMode=BMM_Replace
    scaleSpace=BCS_BoneSpace
    translationMode=BMM_Additive
    translationSpace=BCS_BoneSpace
    alphaBoolBlend {blendOption=Linear}
    alphaScaleBiasClamp {bInterpResult=true, interpSpeedIncreasing=4}
    boneToModify {boneName=dyn_Jacket_l_01_jnt}
    rotation {pitch=15}
    translation {z=2}
   }
   n48 = "Transform (Modify) Bone - Bone: dyn_Jacket_ft_l_01_jnt"(ComponentPose=@n47, Alpha=1, Translation="0,0,2", Rotation="0,-20,0", Scale="1,1,1") {
    rotationMode=BMM_Additive
    rotationSpace=BCS_BoneSpace
    scaleMode=BMM_Replace
    scaleSpace=BCS_BoneSpace
    translationMode=BMM_Additive
    translationSpace=BCS_BoneSpace
    alphaBoolBlend {blendOption=Linear}
    alphaScaleBiasClamp {bInterpResult=true, interpSpeedIncreasing=4}
    boneToModify {boneName=dyn_Jacket_ft_l_01_jnt}
    rotation {yaw=-20}
    translation {z=2}
   }
   n49 = "Transform (Modify) Bone - Bone: dyn_Jacket_bk_c_01_jnt"(ComponentPose=@n48, Alpha=1, Translation="0,0,2", Rotation="80,0,20", Scale="1,1,1") {
    rotationMode=BMM_Additive
    rotationSpace=BCS_BoneSpace
    scaleMode=BMM_Replace
    scaleSpace=BCS_BoneSpace
    translationMode=BMM_Additive
    translationSpace=BCS_BoneSpace
    alphaBoolBlend {blendOption=Linear}
    alphaScaleBiasClamp {bInterpResult=true, interpSpeedIncreasing=4}
    boneToModify {boneName=dyn_Jacket_bk_c_01_jnt}
    rotation {pitch=80, roll=20}
    translation {z=2}
   }
   n50 = "Transform (Modify) Bone - Bone: dyn_Jacket_bk_r_01_jnt"(ComponentPose=@n49, Alpha=1, Translation="0,0,2", Rotation="90,0,0", Scale="1,1,1") {
    rotationMode=BMM_Additive
    rotationSpace=BCS_BoneSpace
    scaleMode=BMM_Replace
    scaleSpace=BCS_BoneSpace
    translationMode=BMM_Additive
    translationSpace=BCS_BoneSpace
    alphaBoolBlend {blendOption=Linear}
    alphaScaleBiasClamp {bInterpResult=true, interpSpeedIncreasing=4}
    boneToModify {boneName=dyn_Jacket_bk_r_01_jnt}
    rotation {pitch=90}
    translation {z=2}
   }
   n51 = "Transform (Modify) Bone - Bone: dyn_Jacket_r_01_jnt"(ComponentPose=@n50, Alpha=1, Translation="0,0,2", Rotation="95,0,0", Scale="1,1,1") {
    rotationMode=BMM_Additive
    rotationSpace=BCS_BoneSpace
    scaleMode=BMM_Replace
    scaleSpace=BCS_BoneSpace
    translationMode=BMM_Additive
    translationSpace=BCS_BoneSpace
    alphaBoolBlend {blendOption=Linear}
    alphaScaleBiasClamp {bInterpResult=true, interpSpeedIncreasing=4}
    boneToModify {boneName=dyn_Jacket_r_01_jnt}
    rotation {pitch=95}
    translation {z=2}
   }
   n52 = "Transform (Modify) Bone - Bone: dyn_Jacket_ft_r_01_jnt"(ComponentPose=@n51, Alpha=1, Translation="0,0,2", Rotation="80,20,0", Scale="1,1,1") {
    rotationMode=BMM_Additive
    rotationSpace=BCS_BoneSpace
    scaleMode=BMM_Replace
    scaleSpace=BCS_BoneSpace
    translationMode=BMM_Additive
    translationSpace=BCS_BoneSpace
    alphaBoolBlend {blendOption=Linear}
    alphaScaleBiasClamp {bInterpResult=true, interpSpeedIncreasing=4}
    boneToModify {boneName=dyn_Jacket_ft_r_01_jnt}
    rotation {pitch=80, yaw=20}
    translation {z=2}
   }
   n53 = "Transform (Modify) Bone - Bone: dyn_Jacket_ft_l_MSTR"(ComponentPose=@n52, Alpha=1, Translation="0,0,1", Rotation="10,0,0", Scale="1,1,1") {
    rotationMode=BMM_Additive
    rotationSpace=BCS_BoneSpace
    scaleMode=BMM_Replace
    scaleSpace=BCS_BoneSpace
    translationMode=BMM_Additive
    translationSpace=BCS_BoneSpace
    alphaBoolBlend {blendOption=Linear}
    alphaScaleBiasClamp {interpSpeedDecreasing=4}
    boneToModify {boneName=dyn_Jacket_ft_l_MSTR}
    rotation {pitch=10}
    translation {z=1}
   }
   n54 = "Transform (Modify) Bone - Bone: dyn_Jacket_ft_r_MSTR"(ComponentPose=@n53, Alpha=1, Translation="0,0,1", Rotation="10,0,0", Scale="1,1,1") {
    rotationMode=BMM_Additive
    rotationSpace=BCS_BoneSpace
    scaleMode=BMM_Replace
    scaleSpace=BCS_BoneSpace
    translationMode=BMM_Additive
    translationSpace=BCS_BoneSpace
    alphaBoolBlend {blendOption=Linear}
    alphaScaleBiasClamp {interpSpeedDecreasing=4}
    boneToModify {boneName=dyn_Jacket_ft_r_MSTR}
    rotation {pitch=10}
    translation {z=1}
   }
   n55 = "Transform (Modify) Bone - Bone: dyn_Jacket_l_MSTR"(ComponentPose=@n54, Alpha=1, Translation="0,0,2", Rotation="20,0,0", Scale="1,1,1") {
    rotationMode=BMM_Additive
    rotationSpace=BCS_BoneSpace
    scaleMode=BMM_Replace
    scaleSpace=BCS_BoneSpace
    translationMode=BMM_Additive
    translationSpace=BCS_BoneSpace
    alphaBoolBlend {blendOption=Linear}
    alphaScaleBiasClamp {interpSpeedDecreasing=4}
    boneToModify {boneName=dyn_Jacket_l_MSTR}
    rotation {pitch=20}
    translation {z=2}
   }
   n56 = "Transform (Modify) Bone - Bone: dyn_Jacket_r_MSTR"(ComponentPose=@n55, Alpha=1, Translation="0,0,2", Rotation="20,0,0", Scale="1,1,1") {
    rotationMode=BMM_Additive
    rotationSpace=BCS_BoneSpace
    scaleMode=BMM_Replace
    scaleSpace=BCS_BoneSpace
    translationMode=BMM_Additive
    translationSpace=BCS_BoneSpace
    alphaBoolBlend {blendOption=Linear}
    alphaScaleBiasClamp {interpSpeedDecreasing=4}
    boneToModify {boneName=dyn_Jacket_r_MSTR}
    rotation {pitch=20}
    translation {z=2}
   }
   n57 = Anim Dynamics - Bone: dyn_Jacket_ft_l_01_jnt(ComponentPose=@n56, Alpha=1, GravityOverride="0,0,-980", ExternalForce="0,0,0") {
    angularDampingOverride=1
    angularSpringConstant=50
    bAngularSpring=true
    bOverrideAngularDamping=true
    bUseGravityOverride=true
    gravityScale=0.35
    simulationSpace=RootRelative
    alphaBoolBlend {blendOption=Linear}
    boundBone {boneName=dyn_Jacket_ft_l_01_jnt}
    componentLinearAccScale=<0.4 0.4 0.4>
    componentLinearVelScale=<0.4 0.4 0.4>
    gravityOverride {z=-980}
    physicsBodyDefinitions [{
      boundBone {boneName=dyn_Jacket_ft_l_01_jnt}
      boxExtents=<14 14 14>
      constraintSetup {angularLimitsMax=<1 0 5>, angularLimitsMin=<-1 -15 -5>, angularTarget=<1 0 0>}
      localJointOffset=<2 0 0>
     }]
   }
   n58 = Anim Dynamics - Bone: dyn_Jacket_ft_l_02_jnt(ComponentPose=@n57, Alpha=1, GravityOverride="0,0,-1000", ExternalForce="0,0,0") {
    angularDampingOverride=1
    angularSpringConstant=50
    bAngularSpring=true
    bOverrideAngularDamping=true
    bUseGravityOverride=true
    gravityScale=0.35
    simulationSpace=RootRelative
    alphaBoolBlend {blendOption=Linear}
    boundBone {boneName=dyn_Jacket_ft_l_02_jnt}
    componentLinearAccScale=<0.4 0.4 0.4>
    componentLinearVelScale=<0.4 0.4 0.4>
    gravityOverride {z=-1000}
    physicsBodyDefinitions [{
      boundBone {boneName=dyn_Jacket_ft_l_02_jnt}
      boxExtents=<14 14 14>
      constraintSetup {angularLimitsMax=<1 0 5>, angularLimitsMin=<-1 -15 -5>, angularTarget=<1 0 0>}
      localJointOffset=<2 0 0>
     }]
   }
   n59 = Anim Dynamics - Bone: dyn_Jacket_ft_l_03_jnt(ComponentPose=@n58, Alpha=1, GravityOverride="0,0,-1000", ExternalForce="0,0,0") {
    angularDampingOverride=1
    angularSpringConstant=50
    bAngularSpring=true
    bOverrideAngularDamping=true
    bUseGravityOverride=true
    gravityScale=0.35
    simulationSpace=RootRelative
    alphaBoolBlend {blendOption=Linear}
    boundBone {boneName=dyn_Jacket_ft_l_03_jnt}
    componentLinearAccScale=<0.4 0.4 0.4>
    componentLinearVelScale=<0.4 0.4 0.4>
    gravityOverride {z=-1000}
    physicsBodyDefinitions [{
      boundBone {boneName=dyn_Jacket_ft_l_03_jnt}
      boxExtents=<14 14 14>
      constraintSetup {angularLimitsMax=<1 0 5>, angularLimitsMin=<-1 -15 -5>, angularTarget=<1 0 0>}
      localJointOffset=<2 0 0>
     }]
   }
   n60 = Anim Dynamics - Bone: dyn_Jacket_ft_l_04_jnt(ComponentPose=@n59, Alpha=1, GravityOverride="0,0,-1000", ExternalForce="0,0,0") {
    angularDampingOverride=1
    angularSpringConstant=50
    bAngularSpring=true
    bOverrideAngularDamping=true
    bUseGravityOverride=true
    gravityScale=0.35
    simulationSpace=RootRelative
    alphaBoolBlend {blendOption=Linear}
    boundBone {boneName=dyn_Jacket_ft_l_04_jnt}
    componentLinearAccScale=<0.4 0.4 0.4>
    componentLinearVelScale=<0.4 0.4 0.4>
    gravityOverride {z=-1000}
    physicsBodyDefinitions [{
      boundBone {boneName=dyn_Jacket_ft_l_04_jnt}
      boxExtents=<14 14 14>
      constraintSetup {angularLimitsMax=<1 0 5>, angularLimitsMin=<-1 -15 -5>, angularTarget=<1 0 0>}
      localJointOffset=<2 0 0>
     }]
   }
   n61 = Anim Dynamics - Bone: dyn_Jacket_l_01_jnt(ComponentPose=@n60, Alpha=1, GravityOverride="0,0,-980", ExternalForce="0,0,0") {
    angularDampingOverride=1
    angularSpringConstant=50
    bAngularSpring=true
    bOverrideAngularDamping=true
    bUseGravityOverride=true
    gravityScale=0.35
    simulationSpace=RootRelative
    alphaBoolBlend {blendOption=Linear}
    boundBone {boneName=dyn_Jacket_l_01_jnt}
    componentLinearAccScale=<0.4 0.4 0.4>
    componentLinearVelScale=<0.4 0.4 0.4>
    gravityOverride {z=-980}
    physicsBodyDefinitions [{
      boundBone {boneName=dyn_Jacket_l_01_jnt}
      boxExtents=<14 14 14>
      constraintSetup {angularLimitsMax=<1 0 5>, angularLimitsMin=<-1 -15 -5>, angularTarget=<1 0 0>}
      localJointOffset=<2 0 0>
     }]
   }
   n62 = Anim Dynamics - Bone: dyn_Jacket_l_02_jnt(ComponentPose=@n61, Alpha=1, GravityOverride="0,0,-1000", ExternalForce="0,0,0") {
    angularDampingOverride=1
    angularSpringConstant=50
    bAngularSpring=true
    bOverrideAngularDamping=true
    bUseGravityOverride=true
    gravityScale=0.35
    simulationSpace=RootRelative
    alphaBoolBlend {blendOption=Linear}
    boundBone {boneName=dyn_Jacket_l_02_jnt}
    componentLinearAccScale=<0.4 0.4 0.4>
    componentLinearVelScale=<0.4 0.4 0.4>
    gravityOverride {z=-1000}
    physicsBodyDefinitions [{
      boundBone {boneName=dyn_Jacket_l_02_jnt}
      boxExtents=<14 14 14>
      constraintSetup {angularLimitsMax=<1 0 5>, angularLimitsMin=<-1 -15 -5>, angularTarget=<1 0 0>}
      localJointOffset=<2 0 0>
     }]
   }
   n63 = Anim Dynamics - Bone: dyn_Jacket_l_03_jnt(ComponentPose=@n62, Alpha=1, GravityOverride="0,0,-1000", ExternalForce="0,0,0") {
    angularDampingOverride=1
    angularSpringConstant=50
    bAngularSpring=true
    bOverrideAngularDamping=true
    bUseGravityOverride=true
    gravityScale=0.35
    simulationSpace=RootRelative
    alphaBoolBlend {blendOption=Linear}
    boundBone {boneName=dyn_Jacket_l_03_jnt}
    componentLinearAccScale=<0.4 0.4 0.4>
    componentLinearVelScale=<0.4 0.4 0.4>
    gravityOverride {z=-1000}
    physicsBodyDefinitions [{
      boundBone {boneName=dyn_Jacket_l_03_jnt}
      boxExtents=<14 14 14>
      constraintSetup {angularLimitsMax=<1 0 5>, angularLimitsMin=<-1 -15 -5>, angularTarget=<1 0 0>}
      localJointOffset=<2 0 0>
     }]
   }
   n64 = Anim Dynamics - Bone: dyn_Jacket_l_04_jnt(ComponentPose=@n63, Alpha=1, GravityOverride="0,0,-1000", ExternalForce="0,0,0") {
    angularDampingOverride=1
    angularSpringConstant=50
    bAngularSpring=true
    bOverrideAngularDamping=true
    bUseGravityOverride=true
    gravityScale=0.35
    simulationSpace=RootRelative
    alphaBoolBlend {blendOption=Linear}
    boundBone {boneName=dyn_Jacket_l_04_jnt}
    componentLinearAccScale=<0.4 0.4 0.4>
    componentLinearVelScale=<0.4 0.4 0.4>
    gravityOverride {z=-1000}
    physicsBodyDefinitions [{
      boundBone {boneName=dyn_Jacket_l_04_jnt}
      boxExtents=<14 14 14>
      constraintSetup {angularLimitsMax=<1 0 5>, angularLimitsMin=<-1 -15 -5>, angularTarget=<1 0 0>}
      localJointOffset=<2 0 0>
     }]
   }
   n65 = Anim Dynamics - Bone: dyn_Jacket_bk_l_01_jnt!(SimulationSpace:AnimPhysSimSpaceType Enum, ComponentPose=@n64, Alpha=1, GravityOverride:Vector="0,0,-980", ExternalForce="0,0,0") -> (Pose) {
    angularDampingOverride=1
    angularSpringConstant=50
    bAngularSpring=true
    bOverrideAngularDamping=true
    bUseGravityOverride=true
    gravityScale=0.35
    simulationSpace=RootRelative
    alphaBoolBlend {blendOption=Linear}
    boundBone {boneName=dyn_Jacket_bk_l_01_jnt}
    componentLinearAccScale=<0.4 0.4 0.4>
    componentLinearVelScale=<0.4 0.4 0.4>
    gravityOverride {z=-980}
    physicsBodyDefinitions [{
      boundBone {boneName=dyn_Jacket_bk_l_01_jnt}
      boxExtents=<14 14 14>
      constraintSetup {angularLimitsMax=<1 0 5>, angularLimitsMin=<-1 -15 -5>, angularTarget=<1 0 0>}
      localJointOffset=<2 0 0>
     }]
   }
   n66 = Anim Dynamics - Bone: dyn_Jacket_bk_l_02_jnt!(SimulationSpace:AnimPhysSimSpaceType Enum, ComponentPose=@n65, Alpha=1, GravityOverride:Vector="0,0,-1000", ExternalForce="0,0,0") -> (Pose) {
    angularDampingOverride=1
    angularSpringConstant=50
    bAngularSpring=true
    bOverrideAngularDamping=true
    bUseGravityOverride=true
    gravityScale=0.35
    simulationSpace=RootRelative
    alphaBoolBlend {blendOption=Linear}
    boundBone {boneName=dyn_Jacket_bk_l_02_jnt}
    componentLinearAccScale=<0.4 0.4 0.4>
    componentLinearVelScale=<0.4 0.4 0.4>
    gravityOverride {z=-1000}
    physicsBodyDefinitions [{
      boundBone {boneName=dyn_Jacket_bk_l_02_jnt}
      boxExtents=<14 14 14>
      constraintSetup {angularLimitsMax=<1 0 5>, angularLimitsMin=<-1 -15 -5>, angularTarget=<1 0 0>}
      localJointOffset=<2 0 0>
     }]
   }
   n67 = Anim Dynamics - Bone: dyn_Jacket_bk_l_03_jnt!(SimulationSpace:AnimPhysSimSpaceType Enum, ComponentPose=@n66, Alpha=1, GravityOverride:Vector="0,0,-1000", ExternalForce="0,0,0") -> (Pose) {
    angularDampingOverride=1
    angularSpringConstant=50
    bAngularSpring=true
    bOverrideAngularDamping=true
    bUseGravityOverride=true
    gravityScale=0.35
    simulationSpace=RootRelative
    alphaBoolBlend {blendOption=Linear}
    boundBone {boneName=dyn_Jacket_bk_l_03_jnt}
    componentLinearAccScale=<0.4 0.4 0.4>
    componentLinearVelScale=<0.4 0.4 0.4>
    gravityOverride {z=-1000}
    physicsBodyDefinitions [{
      boundBone {boneName=dyn_Jacket_bk_l_03_jnt}
      boxExtents=<14 14 14>
      constraintSetup {angularLimitsMax=<1 0 5>, angularLimitsMin=<-1 -15 -5>, angularTarget=<1 0 0>}
      localJointOffset=<2 0 0>
     }]
   }
   n68 = Anim Dynamics - Bone: dyn_Jacket_bk_l_04_jnt!(SimulationSpace:AnimPhysSimSpaceType Enum, ComponentPose=@n67, Alpha=1, GravityOverride:Vector="0,0,-1000", ExternalForce="0,0,0") -> (Pose) {
    angularDampingOverride=1
    angularSpringConstant=50
    bAngularSpring=true
    bOverrideAngularDamping=true
    bUseGravityOverride=true
    gravityScale=0.35
    simulationSpace=RootRelative
    alphaBoolBlend {blendOption=Linear}
    boundBone {boneName=dyn_Jacket_bk_l_04_jnt}
    componentLinearAccScale=<0.4 0.4 0.4>
    componentLinearVelScale=<0.4 0.4 0.4>
    gravityOverride {z=-1000}
    physicsBodyDefinitions [{
      boundBone {boneName=dyn_Jacket_bk_l_04_jnt}
      boxExtents=<14 14 14>
      constraintSetup {angularLimitsMax=<1 0 5>, angularLimitsMin=<-1 -15 -5>, angularTarget=<1 0 0>}
      localJointOffset=<2 0 0>
     }]
   }
   n69 = Anim Dynamics - Bone: dyn_Jacket_ft_r_01_jnt(ComponentPose=@n68, Alpha=1, GravityOverride="0,0,-980", ExternalForce="0,0,0") {
    angularDampingOverride=1
    angularSpringConstant=50
    bAngularSpring=true
    bOverrideAngularDamping=true
    bUseGravityOverride=true
    gravityScale=0.35
    simulationSpace=RootRelative
    alphaBoolBlend {blendOption=Linear}
    boundBone {boneName=dyn_Jacket_ft_r_01_jnt}
    componentLinearAccScale=<0.4 0.4 0.4>
    componentLinearVelScale=<0.4 0.4 0.4>
    gravityOverride {z=-980}
    physicsBodyDefinitions [{
      boundBone {boneName=dyn_Jacket_ft_r_01_jnt}
      boxExtents=<14 14 14>
      constraintSetup {angularLimitsMax=<1 0 5>, angularLimitsMin=<-1 -15 -5>, angularTarget=<1 0 0>}
      localJointOffset=<2 0 0>
     }]
   }
   n70 = Anim Dynamics - Bone: dyn_Jacket_ft_r_02_jnt(ComponentPose=@n69, Alpha=1, GravityOverride="0,0,-1000", ExternalForce="0,0,0") {
    angularDampingOverride=1
    angularSpringConstant=50
    bAngularSpring=true
    bOverrideAngularDamping=true
    bUseGravityOverride=true
    gravityScale=0.35
    simulationSpace=RootRelative
    alphaBoolBlend {blendOption=Linear}
    boundBone {boneName=dyn_Jacket_ft_r_02_jnt}
    componentLinearAccScale=<0.4 0.4 0.4>
    componentLinearVelScale=<0.4 0.4 0.4>
    gravityOverride {z=-1000}
    physicsBodyDefinitions [{
      boundBone {boneName=dyn_Jacket_ft_r_02_jnt}
      boxExtents=<14 14 14>
      constraintSetup {angularLimitsMax=<1 0 5>, angularLimitsMin=<-1 -15 -5>, angularTarget=<1 0 0>}
      localJointOffset=<2 0 0>
     }]
   }
   n71 = Anim Dynamics - Bone: dyn_Jacket_ft_r_03_jnt(ComponentPose=@n70, Alpha=1, GravityOverride="0,0,-1000", ExternalForce="0,0,0") {
    angularDampingOverride=1
    angularSpringConstant=50
    bAngularSpring=true
    bOverrideAngularDamping=true
    bUseGravityOverride=true
    gravityScale=0.35
    simulationSpace=RootRelative
    alphaBoolBlend {blendOption=Linear}
    boundBone {boneName=dyn_Jacket_ft_r_03_jnt}
    componentLinearAccScale=<0.4 0.4 0.4>
    componentLinearVelScale=<0.4 0.4 0.4>
    gravityOverride {z=-1000}
    physicsBodyDefinitions [{
      boundBone {boneName=dyn_Jacket_ft_r_03_jnt}
      boxExtents=<14 14 14>
      constraintSetup {angularLimitsMax=<1 0 5>, angularLimitsMin=<-1 -15 -5>, angularTarget=<1 0 0>}
      localJointOffset=<2 0 0>
     }]
   }
   n72 = Anim Dynamics - Bone: dyn_Jacket_ft_r_04_jnt(ComponentPose=@n71, Alpha=1, GravityOverride="0,0,-1000", ExternalForce="0,0,0") {
    angularDampingOverride=1
    angularSpringConstant=50
    bAngularSpring=true
    bOverrideAngularDamping=true
    bUseGravityOverride=true
    gravityScale=0.35
    simulationSpace=RootRelative
    alphaBoolBlend {blendOption=Linear}
    boundBone {boneName=dyn_Jacket_ft_r_04_jnt}
    componentLinearAccScale=<0.4 0.4 0.4>
    componentLinearVelScale=<0.4 0.4 0.4>
    gravityOverride {z=-1000}
    physicsBodyDefinitions [{
      boundBone {boneName=dyn_Jacket_ft_r_04_jnt}
      boxExtents=<14 14 14>
      constraintSetup {angularLimitsMax=<1 0 5>, angularLimitsMin=<-1 -15 -5>, angularTarget=<1 0 0>}
      localJointOffset=<2 0 0>
     }]
   }
   n73 = Anim Dynamics - Bone: dyn_Jacket_r_01_jnt(ComponentPose=@n72, Alpha=1, GravityOverride="0,0,-980", ExternalForce="0,0,0") {
    angularDampingOverride=1
    angularSpringConstant=50
    bAngularSpring=true
    bOverrideAngularDamping=true
    bUseGravityOverride=true
    gravityScale=0.35
    simulationSpace=RootRelative
    alphaBoolBlend {blendOption=Linear}
    boundBone {boneName=dyn_Jacket_r_01_jnt}
    componentLinearAccScale=<0.4 0.4 0.4>
    componentLinearVelScale=<0.4 0.4 0.4>
    gravityOverride {z=-980}
    physicsBodyDefinitions [{
      boundBone {boneName=dyn_Jacket_r_01_jnt}
      boxExtents=<14 14 14>
      constraintSetup {angularLimitsMax=<1 0 5>, angularLimitsMin=<-1 -15 -5>, angularTarget=<1 0 0>}
      localJointOffset=<2 0 0>
     }]
   }
   n74 = Anim Dynamics - Bone: dyn_Jacket_r_02_jnt(ComponentPose=@n73, Alpha=1, GravityOverride="0,0,-1000", ExternalForce="0,0,0") {
    angularDampingOverride=1
    angularSpringConstant=50
    bAngularSpring=true
    bOverrideAngularDamping=true
    bUseGravityOverride=true
    gravityScale=0.35
    simulationSpace=RootRelative
    alphaBoolBlend {blendOption=Linear}
    boundBone {boneName=dyn_Jacket_r_02_jnt}
    componentLinearAccScale=<0.4 0.4 0.4>
    componentLinearVelScale=<0.4 0.4 0.4>
    gravityOverride {z=-1000}
    physicsBodyDefinitions [{
      boundBone {boneName=dyn_Jacket_r_02_jnt}
      boxExtents=<14 14 14>
      constraintSetup {angularLimitsMax=<1 0 5>, angularLimitsMin=<-1 -15 -5>, angularTarget=<1 0 0>}
      localJointOffset=<2 0 0>
     }]
   }
   n75 = Anim Dynamics - Bone: dyn_Jacket_r_03_jnt(ComponentPose=@n74, Alpha=1, GravityOverride="0,0,-1000", ExternalForce="0,0,0") {
    angularDampingOverride=1
    angularSpringConstant=50
    bAngularSpring=true
    bOverrideAngularDamping=true
    bUseGravityOverride=true
    gravityScale=0.35
    simulationSpace=RootRelative
    alphaBoolBlend {blendOption=Linear}
    boundBone {boneName=dyn_Jacket_r_03_jnt}
    componentLinearAccScale=<0.4 0.4 0.4>
    componentLinearVelScale=<0.4 0.4 0.4>
    gravityOverride {z=-1000}
    physicsBodyDefinitions [{
      boundBone {boneName=dyn_Jacket_r_03_jnt}
      boxExtents=<14 14 14>
      constraintSetup {angularLimitsMax=<1 0 5>, angularLimitsMin=<-1 -15 -5>, angularTarget=<1 0 0>}
      localJointOffset=<2 0 0>
     }]
   }
   n76 = Anim Dynamics - Bone: dyn_Jacket_r_04_jnt(ComponentPose=@n75, Alpha=1, GravityOverride="0,0,-1000", ExternalForce="0,0,0") {
    angularDampingOverride=1
    angularSpringConstant=50
    bAngularSpring=true
    bOverrideAngularDamping=true
    bUseGravityOverride=true
    gravityScale=0.35
    simulationSpace=RootRelative
    alphaBoolBlend {blendOption=Linear}
    boundBone {boneName=dyn_Jacket_r_04_jnt}
    componentLinearAccScale=<0.4 0.4 0.4>
    componentLinearVelScale=<0.4 0.4 0.4>
    gravityOverride {z=-1000}
    physicsBodyDefinitions [{
      boundBone {boneName=dyn_Jacket_r_04_jnt}
      boxExtents=<14 14 14>
      constraintSetup {angularLimitsMax=<1 0 5>, angularLimitsMin=<-1 -15 -5>, angularTarget=<1 0 0>}
      localJointOffset=<2 0 0>
     }]
   }
   n77 = Anim Dynamics - Bone: dyn_Jacket_bk_r_01_jnt!(SimulationSpace:AnimPhysSimSpaceType Enum, ComponentPose=@n76, Alpha=1, GravityOverride:Vector="0,0,-980", ExternalForce="0,0,0") -> (Pose) {
    angularDampingOverride=1
    angularSpringConstant=50
    bAngularSpring=true
    bOverrideAngularDamping=true
    bUseGravityOverride=true
    gravityScale=0.35
    simulationSpace=RootRelative
    alphaBoolBlend {blendOption=Linear}
    boundBone {boneName=dyn_Jacket_bk_r_01_jnt}
    componentLinearAccScale=<0.4 0.4 0.4>
    componentLinearVelScale=<0.4 0.4 0.4>
    gravityOverride {z=-980}
    physicsBodyDefinitions [{
      boundBone {boneName=dyn_Jacket_bk_r_01_jnt}
      boxExtents=<14 14 14>
      constraintSetup {angularLimitsMax=<1 0 5>, angularLimitsMin=<-1 -15 -5>, angularTarget=<1 0 0>}
      localJointOffset=<2 0 0>
     }]
   }
   n78 = Anim Dynamics - Bone: dyn_Jacket_bk_r_02_jnt!(SimulationSpace:AnimPhysSimSpaceType Enum, ComponentPose=@n77, Alpha=1, GravityOverride:Vector="0,0,-1000", ExternalForce="0,0,0") -> (Pose) {
    angularDampingOverride=1
    angularSpringConstant=50
    bAngularSpring=true
    bOverrideAngularDamping=true
    bUseGravityOverride=true
    gravityScale=0.35
    simulationSpace=RootRelative
    alphaBoolBlend {blendOption=Linear}
    boundBone {boneName=dyn_Jacket_bk_r_02_jnt}
    componentLinearAccScale=<0.4 0.4 0.4>
    componentLinearVelScale=<0.4 0.4 0.4>
    gravityOverride {z=-1000}
    physicsBodyDefinitions [{
      boundBone {boneName=dyn_Jacket_bk_r_02_jnt}
      boxExtents=<14 14 14>
      constraintSetup {angularLimitsMax=<1 0 5>, angularLimitsMin=<-1 -15 -5>, angularTarget=<1 0 0>}
      localJointOffset=<2 0 0>
     }]
   }
   n79 = Anim Dynamics - Bone: dyn_Jacket_bk_r_03_jnt!(SimulationSpace:AnimPhysSimSpaceType Enum, ComponentPose=@n78, Alpha=1, GravityOverride:Vector="0,0,-1000", ExternalForce="0,0,0") -> (Pose) {
    angularDampingOverride=1
    angularSpringConstant=50
    bAngularSpring=true
    bOverrideAngularDamping=true
    bUseGravityOverride=true
    gravityScale=0.35
    simulationSpace=RootRelative
    alphaBoolBlend {blendOption=Linear}
    boundBone {boneName=dyn_Jacket_bk_r_03_jnt}
    componentLinearAccScale=<0.4 0.4 0.4>
    componentLinearVelScale=<0.4 0.4 0.4>
    gravityOverride {z=-1000}
    physicsBodyDefinitions [{
      boundBone {boneName=dyn_Jacket_bk_r_03_jnt}
      boxExtents=<14 14 14>
      constraintSetup {angularLimitsMax=<1 0 5>, angularLimitsMin=<-1 -15 -5>, angularTarget=<1 0 0>}
      localJointOffset=<2 0 0>
     }]
   }
   n80 = Anim Dynamics - Bone: dyn_Jacket_bk_r_04_jnt!(SimulationSpace:AnimPhysSimSpaceType Enum, ComponentPose=@n79, Alpha=1, GravityOverride:Vector="0,0,-1000", ExternalForce="0,0,0") -> (Pose) {
    angularDampingOverride=1
    angularSpringConstant=50
    bAngularSpring=true
    bOverrideAngularDamping=true
    bUseGravityOverride=true
    gravityScale=0.35
    simulationSpace=RootRelative
    alphaBoolBlend {blendOption=Linear}
    boundBone {boneName=dyn_Jacket_bk_r_04_jnt}
    componentLinearAccScale=<0.4 0.4 0.4>
    componentLinearVelScale=<0.4 0.4 0.4>
    gravityOverride {z=-1000}
    physicsBodyDefinitions [{
      boundBone {boneName=dyn_Jacket_bk_r_04_jnt}
      boxExtents=<14 14 14>
      constraintSetup {angularLimitsMax=<1 0 5>, angularLimitsMin=<-1 -15 -5>, angularTarget=<1 0 0>}
      localJointOffset=<2 0 0>
     }]
   }
   n81 = Anim Dynamics - Bone: dyn_Jacket_bk_c_01_jnt!(SimulationSpace:AnimPhysSimSpaceType Enum, ComponentPose=@n80, Alpha=1, GravityOverride:Vector="0,0,-980", ExternalForce="0,0,0") -> (Pose) {
    angularDampingOverride=1
    angularSpringConstant=50
    bAngularSpring=true
    bOverrideAngularDamping=true
    bUseGravityOverride=true
    gravityScale=0.35
    simulationSpace=RootRelative
    alphaBoolBlend {blendOption=Linear}
    boundBone {boneName=dyn_Jacket_bk_c_01_jnt}
    componentLinearAccScale=<0.4 0.4 0.4>
    componentLinearVelScale=<0.4 0.4 0.4>
    gravityOverride {z=-980}
    physicsBodyDefinitions [{
      boundBone {boneName=dyn_Jacket_bk_c_01_jnt}
      boxExtents=<17 17 17>
      constraintSetup {angularLimitsMax=<1 0 5>, angularLimitsMin=<-1 -15 -5>, angularTarget=<1 0 0>}
      localJointOffset=<2 0 0>
     }]
   }
   n82 = Anim Dynamics - Bone: dyn_Jacket_bk_c_02_jnt!(SimulationSpace:AnimPhysSimSpaceType Enum, ComponentPose=@n81, Alpha=1, GravityOverride:Vector="0,0,-1000", ExternalForce="0,0,0") -> (Pose) {
    angularDampingOverride=1
    angularSpringConstant=50
    bAngularSpring=true
    bOverrideAngularDamping=true
    bUseGravityOverride=true
    gravityScale=0.35
    simulationSpace=RootRelative
    alphaBoolBlend {blendOption=Linear}
    boundBone {boneName=dyn_Jacket_bk_c_02_jnt}
    componentLinearAccScale=<0.4 0.4 0.4>
    componentLinearVelScale=<0.4 0.4 0.4>
    gravityOverride {z=-1000}
    physicsBodyDefinitions [{
      boundBone {boneName=dyn_Jacket_bk_c_02_jnt}
      boxExtents=<17 17 17>
      constraintSetup {angularLimitsMax=<1 0 5>, angularLimitsMin=<-1 -15 -5>, angularTarget=<1 0 0>}
      localJointOffset=<2 0 0>
     }]
   }
   n83 = Anim Dynamics - Bone: dyn_Jacket_bk_c_03_jnt!(SimulationSpace:AnimPhysSimSpaceType Enum, ComponentPose=@n82, Alpha=1, GravityOverride:Vector="0,0,-1000", ExternalForce="0,0,0") -> (Pose) {
    angularDampingOverride=1
    angularSpringConstant=50
    bAngularSpring=true
    bOverrideAngularDamping=true
    bUseGravityOverride=true
    gravityScale=0.35
    simulationSpace=RootRelative
    alphaBoolBlend {blendOption=Linear}
    boundBone {boneName=dyn_Jacket_bk_c_03_jnt}
    componentLinearAccScale=<0.4 0.4 0.4>
    componentLinearVelScale=<0.4 0.4 0.4>
    gravityOverride {z=-1000}
    physicsBodyDefinitions [{
      boundBone {boneName=dyn_Jacket_bk_c_03_jnt}
      boxExtents=<17 17 17>
      constraintSetup {angularLimitsMax=<1 0 5>, angularLimitsMin=<-1 -15 -5>, angularTarget=<1 0 0>}
      localJointOffset=<2 0 0>
     }]
   }
   n84 = Anim Dynamics - Bone: dyn_Jacket_bk_c_04_jnt!(SimulationSpace:AnimPhysSimSpaceType Enum, ComponentPose=@n83, Alpha=1, GravityOverride:Vector="0,0,-1000", ExternalForce="0,0,0") -> (Pose) {
    angularDampingOverride=1
    angularSpringConstant=50
    bAngularSpring=true
    bOverrideAngularDamping=true
    bUseGravityOverride=true
    gravityScale=0.35
    simulationSpace=RootRelative
    alphaBoolBlend {blendOption=Linear}
    boundBone {boneName=dyn_Jacket_bk_c_04_jnt}
    componentLinearAccScale=<0.4 0.4 0.4>
    componentLinearVelScale=<0.4 0.4 0.4>
    gravityOverride {z=-1000}
    physicsBodyDefinitions [{
      boundBone {boneName=dyn_Jacket_bk_c_04_jnt}
      boxExtents=<15 15 15>
      constraintSetup {angularLimitsMax=<1 0 5>, angularLimitsMin=<-1 -15 -5>, angularTarget=<1 0 0>}
      localJointOffset=<2 0 0>
     }]
   }
   n85 = Component To Local(ComponentPose=@n84)
   n86 = Get bClothEnabled()
   n87 = "Transform (Modify) Bone - Bone: dyn_Jacket_bk_r_MSTR"(
    ComponentPose=Local To Component(LocalPose=Blend Poses by bool(BlendPose_0=@n32, BlendPose_1=@n85, bActiveValue=@n86) {blendType=Linear})
    Alpha=1
    Translation="0,0,0"
    Rotation="0,0,0"
    Scale="0.4,1,1"
   ) {
    rotationMode=BMM_Additive
    rotationSpace=BCS_BoneSpace
    scaleMode=BMM_Replace
    scaleSpace=BCS_BoneSpace
    translationMode=BMM_Additive
    translationSpace=BCS_BoneSpace
    alphaBoolBlend {blendOption=Linear}
    alphaScaleBiasClamp {interpSpeedDecreasing=4}
    boneToModify {boneName=dyn_Jacket_bk_r_MSTR}
    scale {x=0.4}
   }
   n88 = "Transform (Modify) Bone - Bone: dyn_Jacket_r_MSTR"(ComponentPose=@n87, Alpha=1, Translation="0,0,0", Rotation="0,0,0", Scale="0.4,1,1") {
    rotationMode=BMM_Additive
    rotationSpace=BCS_BoneSpace
    scaleMode=BMM_Replace
    scaleSpace=BCS_BoneSpace
    translationMode=BMM_Additive
    translationSpace=BCS_BoneSpace
    alphaBoolBlend {blendOption=Linear}
    alphaScaleBiasClamp {interpSpeedDecreasing=4}
    boneToModify {boneName=dyn_Jacket_r_MSTR}
    scale {x=0.4}
   }
   n89 = "Transform (Modify) Bone - Bone: dyn_Jacket_bk_l_MSTR"(ComponentPose=@n88, Alpha=1, Translation="0,0,0", Rotation="0,0,0", Scale="0.4,1,1") {
    rotationMode=BMM_Additive
    rotationSpace=BCS_BoneSpace
    scaleMode=BMM_Replace
    scaleSpace=BCS_BoneSpace
    translationMode=BMM_Additive
    translationSpace=BCS_BoneSpace
    alphaBoolBlend {blendOption=Linear}
    alphaScaleBiasClamp {interpSpeedDecreasing=4}
    boneToModify {boneName=dyn_Jacket_bk_l_MSTR}
    scale {x=0.4}
   }
   n90 = "Transform (Modify) Bone - Bone: dyn_Jacket_l_MSTR"(ComponentPose=@n89, Alpha=1, Translation="0,0,0", Rotation="0,0,0", Scale="0.4,1,1") {
    rotationMode=BMM_Additive
    rotationSpace=BCS_BoneSpace
    scaleMode=BMM_Replace
    scaleSpace=BCS_BoneSpace
    translationMode=BMM_Additive
    translationSpace=BCS_BoneSpace
    alphaBoolBlend {blendOption=Linear}
    alphaScaleBiasClamp {interpSpeedDecreasing=4}
    boneToModify {boneName=dyn_Jacket_l_MSTR}
    scale {x=0.4}
   }
   n91 = "Transform (Modify) Bone - Bone: dyn_Jacket_bk_c_MSTR"(ComponentPose=@n90, Alpha=1, Translation="0,0,0", Rotation="0,0,0", Scale="0.4,1,1") {
    rotationMode=BMM_Additive
    rotationSpace=BCS_BoneSpace
    scaleMode=BMM_Replace
    scaleSpace=BCS_BoneSpace
    translationMode=BMM_Additive
    translationSpace=BCS_BoneSpace
    alphaBoolBlend {blendOption=Linear}
    alphaScaleBiasClamp {interpSpeedDecreasing=4}
    boneToModify {boneName=dyn_Jacket_bk_c_MSTR}
    scale {x=0.4}
   }
   n92 = "Transform (Modify) Bone - Bone: dyn_Jacket_ft_l_MSTR"(ComponentPose=@n91, Alpha=1, Translation="0,0,0", Rotation="0,0,0", Scale="0.4,1,1") {
    rotationMode=BMM_Additive
    rotationSpace=BCS_BoneSpace
    scaleMode=BMM_Replace
    scaleSpace=BCS_BoneSpace
    translationMode=BMM_Additive
    translationSpace=BCS_BoneSpace
    alphaBoolBlend {blendOption=Linear}
    alphaScaleBiasClamp {interpSpeedDecreasing=4}
    boneToModify {boneName=dyn_Jacket_ft_l_MSTR}
    scale {x=0.4}
   }
   n93 = "Transform (Modify) Bone - Bone: dyn_Jacket_ft_r_MSTR"(ComponentPose=@n92, Alpha=1, Translation="0,0,0", Rotation="0,0,0", Scale="0.4,1,1") {
    rotationMode=BMM_Additive
    rotationSpace=BCS_BoneSpace
    scaleMode=BMM_Replace
    scaleSpace=BCS_BoneSpace
    translationMode=BMM_Additive
    translationSpace=BCS_BoneSpace
    alphaBoolBlend {blendOption=Linear}
    alphaScaleBiasClamp {interpSpeedDecreasing=4}
    boneToModify {boneName=dyn_Jacket_ft_r_MSTR}
    scale {x=0.4}
   }
   n94 = "Constraint - dyn_Jacket_bk_l_02_jnt (PP)"!(ComponentPose=@n93, Alpha=1, ConstraintWeights_0=0.5, ConstraintWeights_1:float=0.5) -> (Pose) {
    alphaBoolBlend {blendOption=Linear}
    alphaScaleBiasClamp {scale=0.2}
    boneToModify {boneName=dyn_Jacket_bk_l_02_jnt}
    constraintSetup [{targetBone {boneName=dyn_Jacket_l_02_jnt}}, {targetBone {boneName=dyn_Jacket_bk_c_02_jnt}}]
    constraintWeights [1,1]
   }
   n95 = "Constraint - dyn_Jacket_bk_l_03_jnt (PP)"!(ComponentPose=@n94, Alpha=1, ConstraintWeights_0=0.5, ConstraintWeights_1:float=0.5) -> (Pose) {
    alphaBoolBlend {blendOption=Linear}
    alphaScaleBiasClamp {scale=0.3}
    boneToModify {boneName=dyn_Jacket_bk_l_03_jnt}
    constraintSetup [{targetBone {boneName=dyn_Jacket_l_03_jnt}}, {targetBone {boneName=dyn_Jacket_bk_c_03_jnt}}]
    constraintWeights [1,1]
   }
   n96 = "Constraint - dyn_Jacket_bk_l_04_jnt (PP)"!(ComponentPose=@n95, Alpha=1, ConstraintWeights_0=0.5, ConstraintWeights_1:float=0.5) -> (Pose) {
    alphaBoolBlend {blendOption=Linear}
    alphaScaleBiasClamp {scale=0.2}
    boneToModify {boneName=dyn_Jacket_bk_l_04_jnt}
    constraintSetup [{targetBone {boneName=dyn_Jacket_l_04_jnt}}, {targetBone {boneName=dyn_Jacket_bk_c_04_jnt}}]
    constraintWeights [1,1]
   }
   n97 = "Constraint - dyn_Jacket_bk_r_02_jnt (PP)"!(ComponentPose=@n96, Alpha=1, ConstraintWeights_0=0.5, ConstraintWeights_1:float=0.5) -> (Pose) {
    alphaBoolBlend {blendOption=Linear}
    alphaScaleBiasClamp {scale=0.2}
    boneToModify {boneName=dyn_Jacket_bk_r_02_jnt}
    constraintSetup [{targetBone {boneName=dyn_Jacket_r_02_jnt}}, {targetBone {boneName=dyn_Jacket_bk_c_02_jnt}}]
    constraintWeights [1,1]
   }
   n98 = "Constraint - dyn_Jacket_bk_r_03_jnt (PP)"!(ComponentPose=@n97, Alpha=1, ConstraintWeights_0=0.5, ConstraintWeights_1:float=0.5) -> (Pose) {
    alphaBoolBlend {blendOption=Linear}
    alphaScaleBiasClamp {scale=0.3}
    boneToModify {boneName=dyn_Jacket_bk_r_03_jnt}
    constraintSetup [{targetBone {boneName=dyn_Jacket_r_03_jnt}}, {targetBone {boneName=dyn_Jacket_bk_c_03_jnt}}]
    constraintWeights [1,1]
   }
   n99 = "Constraint - dyn_Jacket_bk_r_04_jnt (PP)"!(ComponentPose=@n98, Alpha=1, ConstraintWeights_0=0.5, ConstraintWeights_1:float=0.5) -> (Pose) {
    alphaBoolBlend {blendOption=Linear}
    alphaScaleBiasClamp {scale=0.2}
    boneToModify {boneName=dyn_Jacket_bk_r_04_jnt}
    constraintSetup [{targetBone {boneName=dyn_Jacket_r_04_jnt}}, {targetBone {boneName=dyn_Jacket_bk_c_04_jnt}}]
    constraintWeights [1,1]
   }
   n100 = "Constraint - dyn_Jacket_bk_c_02_jnt (PP)"!(ComponentPose=@n99, Alpha=1, ConstraintWeights_0=0.5, ConstraintWeights_1:float=0.5) -> (Pose) {
    alphaBoolBlend {blendOption=Linear}
    alphaScaleBiasClamp {scale=0.2}
    boneToModify {boneName=dyn_Jacket_bk_c_02_jnt}
    constraintSetup [{targetBone {boneName=dyn_Jacket_bk_l_02_jnt}}, {targetBone {boneName=dyn_Jacket_bk_r_02_jnt}}]
    constraintWeights [1,1]
   }
   n101 = "Constraint - dyn_Jacket_bk_c_03_jnt (PP)"!(ComponentPose=@n100, Alpha=1, ConstraintWeights_0=0.5, ConstraintWeights_1:float=0.5) -> (Pose) {
    alphaBoolBlend {blendOption=Linear}
    alphaScaleBiasClamp {scale=0.3}
    boneToModify {boneName=dyn_Jacket_bk_c_03_jnt}
    constraintSetup [{targetBone {boneName=dyn_Jacket_bk_l_03_jnt}}, {targetBone {boneName=dyn_Jacket_bk_r_03_jnt}}]
    constraintWeights [1,1]
   }
   n102 = "Constraint - dyn_Jacket_bk_c_04_jnt (PP)"!(ComponentPose=@n101, Alpha=1, ConstraintWeights_0=0.5, ConstraintWeights_1:float=0.5) -> (Pose) {
    alphaBoolBlend {blendOption=Linear}
    alphaScaleBiasClamp {scale=0.2}
    boneToModify {boneName=dyn_Jacket_bk_c_04_jnt}
    constraintSetup [{targetBone {boneName=dyn_Jacket_bk_l_04_jnt}}, {targetBone {boneName=dyn_Jacket_bk_r_04_jnt}}]
    constraintWeights [1,1]
   }
   Output Pose(Result=Component To Local(ComponentPose=@n102)) {layerGroup=None}
   n103 = "Constraint - dyn_Jacket_flap_l_MSTR (R)"(ComponentPose=Local To Component(LocalPose=Copy Pose From Mesh() {bUseAttachedParent=true}), Alpha=1, ConstraintWeights_0=1) {
    alphaBoolBlend {blendOption=Linear}
    boneToModify {boneName=dyn_Jacket_flap_l_MSTR}
    constraintSetup [{transformType=Rotation, targetBone {boneName=clavicle_l}}]
    constraintWeights [1]
   }
   n104 = "Constraint - dyn_Jacket_flap_r_MSTR (R)"(ComponentPose=@n103, Alpha=1, ConstraintWeights_0=1) {
    alphaBoolBlend {blendOption=Linear}
    boneToModify {boneName=dyn_Jacket_flap_r_MSTR}
    constraintSetup [{transformType=Rotation, targetBone {boneName=clavicle_r}}]
    constraintWeights [1]
   }
   n105 = "Constraint - dyn_Jacket_lowerflap_l_MSTR (R)"(ComponentPose=@n104, Alpha=1, ConstraintWeights_0=1) {
    alphaBoolBlend {blendOption=Linear}
    boneToModify {boneName=dyn_Jacket_lowerflap_l_MSTR}
    constraintSetup [{transformType=Rotation, targetBone {boneName=clavicle_l}}]
    constraintWeights [1]
   }
   n106 = "Constraint - dyn_Jacket_lowerflap_r_MSTR (R)"(ComponentPose=@n105, Alpha=1, ConstraintWeights_0=1) {
    alphaBoolBlend {blendOption=Linear}
    boneToModify {boneName=dyn_Jacket_lowerflap_r_MSTR}
    constraintSetup [{transformType=Rotation, targetBone {boneName=clavicle_r}}]
    constraintWeights [1]
   }
   n107 = "Constraint - dyn_Jacket_upperflap_l_MSTR (R)"(ComponentPose=@n106, Alpha=1, ConstraintWeights_0=1) {
    alphaBoolBlend {blendOption=Linear}
    boneToModify {boneName=dyn_Jacket_upperflap_l_MSTR}
    constraintSetup [{transformType=Rotation, targetBone {boneName=clavicle_l}}]
    constraintWeights [1]
   }
   n108 = "Constraint - dyn_Jacket_upperflap_r_MSTR (R)"(ComponentPose=@n107, Alpha=1, ConstraintWeights_0=1) {
    alphaBoolBlend {blendOption=Linear}
    boneToModify {boneName=dyn_Jacket_upperflap_r_MSTR}
    constraintSetup [{transformType=Rotation, targetBone {boneName=clavicle_r}}]
    constraintWeights [1]
   }
   n109 = Control Rig!(
    Source=Component To Local(ComponentPose=@n108)
    Alpha
    Movement Alpha:double=Get MovementAlpha() -> (MovementAlpha:double)
    is_DBNO_Carried:bool=Get bIsDBNOCarried()
    Is Emoting Alpha:double=Get Is Emoting Alpha() -> (Is Emoting Alpha:double)
    Is Pistol Alpha:double=Get Is Pistol Alpha() -> (Is Pistol Alpha:float)
    Is Rifle Alpha:double=Get Is Rifle Alpha() -> (Is Rifle Alpha:float)
    Is not Emoting Alpha:double=Get Is not Emoting Alpha() -> (Is not Emoting Alpha:double)
   ) -> (Pose) {
    alphaBoolBlend {blendOption=Linear}
    controlRigAssetReference {blueprintRigClass=&/M_MED_FlowerVase_CtrlRig.M_MED_FlowerVase_CtrlRig_C}
   }
   n110 = Anim Dynamics - Bone: dyn_Jacket_flap_l_01_jnt!(ComponentPose=Local To Component(LocalPose=@n109), Alpha=0.25, GravityOverride:Vector="0,0,0", ExternalForce="0,0,0") -> (Pose) {
    angularDampingOverride=1
    angularSpringConstant=20
    bAngularSpring=true
    bOverrideAngularDamping=true
    bUseGravityOverride=true
    simulationSpace=RootRelative
    alphaBoolBlend {blendOption=Linear}
    boundBone {boneName=dyn_Jacket_flap_l_01_jnt}
    componentAppliedLinearAccClamp=<6000 6000 6000>
    componentLinearAccScale=<0.4 0.4 0.4>
    componentLinearVelScale=<0.4 0.4 0.4>
    physicsBodyDefinitions [{
      boundBone {boneName=dyn_Jacket_flap_l_01_jnt}
      boxExtents=<10 10 10>
      constraintSetup {angularLimitsMax=<2 10 2>, angularLimitsMin=<-2 -10 -2>, angularTarget=<1 0 0>}
      localJointOffset=<3 0 0>
     }]
   }
   n111 = Anim Dynamics - Bone: dyn_Jacket_flap_r_01_jnt!(ComponentPose=@n110, Alpha=0.25, GravityOverride:Vector="0,0,0", ExternalForce="0,0,0") -> (Pose) {
    angularDampingOverride=1
    angularSpringConstant=20
    bAngularSpring=true
    bOverrideAngularDamping=true
    bUseGravityOverride=true
    simulationSpace=RootRelative
    alphaBoolBlend {blendOption=Linear}
    boundBone {boneName=dyn_Jacket_flap_r_01_jnt}
    componentAppliedLinearAccClamp=<6000 6000 6000>
    componentLinearAccScale=<0.4 0.4 0.4>
    componentLinearVelScale=<0.4 0.4 0.4>
    physicsBodyDefinitions [{
      boundBone {boneName=dyn_Jacket_flap_r_01_jnt}
      boxExtents=<10 10 10>
      constraintSetup {angularLimitsMax=<2 10 2>, angularLimitsMin=<-2 -10 -2>, angularTarget=<1 0 0>}
      localJointOffset=<3 0 0>
     }]
   }
   n112 = Anim Dynamics - Bone: dyn_Jacket_lowerflap_l_01_jnt!(ComponentPose=@n111, Alpha=0.25, GravityOverride:Vector="0,0,0", ExternalForce="0,0,0") -> (Pose) {
    angularDampingOverride=1
    angularSpringConstant=20
    bAngularSpring=true
    bOverrideAngularDamping=true
    bUseGravityOverride=true
    simulationSpace=RootRelative
    alphaBoolBlend {blendOption=Linear}
    boundBone {boneName=dyn_Jacket_lowerflap_l_01_jnt}
    componentAppliedLinearAccClamp=<6000 6000 6000>
    componentLinearAccScale=<0.4 0.4 0.4>
    componentLinearVelScale=<0.4 0.4 0.4>
    physicsBodyDefinitions [{
      boundBone {boneName=dyn_Jacket_lowerflap_l_01_jnt}
      boxExtents=<10 10 10>
      constraintSetup {angularLimitsMax=<2 10 2>, angularLimitsMin=<-2 -10 -2>, angularTarget=<1 0 0>}
      localJointOffset=<3 0 0>
     }]
   }
   n113 = Anim Dynamics - Bone: dyn_Jacket_lowerflap_r_01_jnt!(ComponentPose=@n112, Alpha=0.25, GravityOverride:Vector="0,0,0", ExternalForce="0,0,0") -> (Pose) {
    angularDampingOverride=1
    angularSpringConstant=20
    bAngularSpring=true
    bOverrideAngularDamping=true
    bUseGravityOverride=true
    simulationSpace=RootRelative
    alphaBoolBlend {blendOption=Linear}
    boundBone {boneName=dyn_Jacket_lowerflap_r_01_jnt}
    componentAppliedLinearAccClamp=<6000 6000 6000>
    componentLinearAccScale=<0.4 0.4 0.4>
    componentLinearVelScale=<0.4 0.4 0.4>
    physicsBodyDefinitions [{
      boundBone {boneName=dyn_Jacket_lowerflap_r_01_jnt}
      boxExtents=<10 10 10>
      constraintSetup {angularLimitsMax=<2 10 2>, angularLimitsMin=<-2 -10 -2>, angularTarget=<1 0 0>}
      localJointOffset=<3 0 0>
     }]
   }
   n114 = Anim Dynamics - Bone: dyn_Jacket_upperflap_l_01_jnt!(ComponentPose=@n113, Alpha=0.25, GravityOverride:Vector="0,0,0", ExternalForce="0,0,0") -> (Pose) {
    angularDampingOverride=1
    angularSpringConstant=20
    bAngularSpring=true
    bOverrideAngularDamping=true
    bUseGravityOverride=true
    simulationSpace=RootRelative
    alphaBoolBlend {blendOption=Linear}
    boundBone {boneName=dyn_Jacket_upperflap_l_01_jnt}
    componentAppliedLinearAccClamp=<6000 6000 6000>
    componentLinearAccScale=<0.4 0.4 0.4>
    componentLinearVelScale=<0.4 0.4 0.4>
    physicsBodyDefinitions [{
      boundBone {boneName=dyn_Jacket_upperflap_l_01_jnt}
      boxExtents=<10 10 10>
      constraintSetup {angularLimitsMax=<2 10 2>, angularLimitsMin=<-2 -10 -2>, angularTarget=<1 0 0>}
      localJointOffset=<3 0 0>
     }]
   }
   n115 = Anim Dynamics - Bone: dyn_Jacket_upperflap_r_01_jnt!(ComponentPose=@n114, Alpha=0.25, GravityOverride:Vector="0,0,0", ExternalForce="0,0,0") -> (Pose) {
    angularDampingOverride=1
    angularSpringConstant=20
    bAngularSpring=true
    bOverrideAngularDamping=true
    bUseGravityOverride=true
    simulationSpace=RootRelative
    alphaBoolBlend {blendOption=Linear}
    boundBone {boneName=dyn_Jacket_upperflap_r_01_jnt}
    componentAppliedLinearAccClamp=<6000 6000 6000>
    componentLinearAccScale=<0.4 0.4 0.4>
    componentLinearVelScale=<0.4 0.4 0.4>
    physicsBodyDefinitions [{
      boundBone {boneName=dyn_Jacket_upperflap_r_01_jnt}
      boxExtents=<10 10 10>
      constraintSetup {angularLimitsMax=<2 10 2>, angularLimitsMin=<-2 -10 -2>, angularTarget=<1 0 0>}
      localJointOffset=<3 0 0>
     }]
   }
   n116 = Anim Dynamics - Bone: dyn_Sword_01_jnt!(ComponentPose=@n115, Alpha=0.5, GravityOverride:Vector="0,0,0", ExternalForce="0,0,0") -> (Pose) {
    angularDampingOverride=1
    angularSpringConstant=20
    bAngularSpring=true
    bOverrideAngularDamping=true
    bUseGravityOverride=true
    simulationSpace=RootRelative
    alphaBoolBlend {blendOption=Linear}
    boundBone {boneName=dyn_Sword_01_jnt}
    componentAppliedLinearAccClamp=<6000 6000 6000>
    componentLinearAccScale=<0.4 0.4 0.4>
    componentLinearVelScale=<0.4 0.4 0.4>
    physicsBodyDefinitions [{
      boundBone {boneName=dyn_Sword_01_jnt}
      boxExtents=<10 10 10>
      constraintSetup {angularLimitsMax=<2 5 5>, angularLimitsMin=<-2 -5 -5>, angularTarget=<1 0 0>}
      localJointOffset=<3 0 0>
     }]
   }
   n117 = Anim Dynamics - Bone: dyn_shirtruffle_tp_l_jnt!(ComponentPose=@n116, Alpha=0.5, GravityOverride:Vector="0,0,0", ExternalForce="0,0,0") -> (Pose) {
    angularDampingOverride=0.9999
    angularSpringConstant=2000
    bAngularSpring=true
    bOverrideAngularDamping=true
    bUseGravityOverride=true
    simulationSpace=RootRelative
    alphaBoolBlend {blendOption=Linear}
    boundBone {boneName=dyn_shirtruffle_tp_l_jnt}
    componentAppliedLinearAccClamp=<6000 6000 6000>
    componentLinearAccScale=<0.4 0.4 0.4>
    componentLinearVelScale=<0.4 0.4 0.4>
    physicsBodyDefinitions [{
      boundBone {boneName=dyn_shirtruffle_tp_l_jnt}
      boxExtents=<10 10 10>
      constraintSetup {angularLimitsMax=<3 10 3>, angularLimitsMin=<-3 -10 -3>, angularTarget=<1 0 0>}
      localJointOffset=<12 0 0>
     }]
   }
   n118 = Anim Dynamics - Bone: dyn_shirtruffle_tp_r_jnt!(ComponentPose=@n117, Alpha=0.5, GravityOverride:Vector="0,0,0", ExternalForce="0,0,0") -> (Pose) {
    angularDampingOverride=0.9999
    angularSpringConstant=2000
    bAngularSpring=true
    bOverrideAngularDamping=true
    bUseGravityOverride=true
    simulationSpace=RootRelative
    alphaBoolBlend {blendOption=Linear}
    boundBone {boneName=dyn_shirtruffle_tp_r_jnt}
    componentAppliedLinearAccClamp=<6000 6000 6000>
    componentLinearAccScale=<0.4 0.4 0.4>
    componentLinearVelScale=<0.4 0.4 0.4>
    physicsBodyDefinitions [{
      boundBone {boneName=dyn_shirtruffle_tp_r_jnt}
      boxExtents=<10 10 10>
      constraintSetup {angularLimitsMax=<3 10 3>, angularLimitsMin=<-3 -10 -3>, angularTarget=<1 0 0>}
      localJointOffset=<12 0 0>
     }]
   }
   n119 = Anim Dynamics - Bone: dyn_shirtruffle_bt_l_jnt!(ComponentPose=@n118, Alpha=0.5, GravityOverride:Vector="0,0,0", ExternalForce="0,0,0") -> (Pose) {
    angularDampingOverride=0.9999
    angularSpringConstant=2000
    bAngularSpring=true
    bOverrideAngularDamping=true
    bUseGravityOverride=true
    simulationSpace=RootRelative
    alphaBoolBlend {blendOption=Linear}
    boundBone {boneName=dyn_shirtruffle_bt_l_jnt}
    componentAppliedLinearAccClamp=<6000 6000 6000>
    componentLinearAccScale=<0.4 0.4 0.4>
    componentLinearVelScale=<0.4 0.4 0.4>
    physicsBodyDefinitions [{
      boundBone {boneName=dyn_shirtruffle_bt_l_jnt}
      boxExtents=<10 10 10>
      constraintSetup {angularLimitsMax=<3 10 3>, angularLimitsMin=<-3 -10 -3>, angularTarget=<1 0 0>}
      localJointOffset=<12 0 0>
     }]
   }
   n120 = Anim Dynamics - Bone: dyn_shirtruffle_bt_r_jnt!(ComponentPose=@n119, Alpha=0.5, GravityOverride:Vector="0,0,0", ExternalForce="0,0,0") -> (Pose) {
    angularDampingOverride=0.9999
    angularSpringConstant=2000
    bAngularSpring=true
    bOverrideAngularDamping=true
    bUseGravityOverride=true
    simulationSpace=RootRelative
    alphaBoolBlend {blendOption=Linear}
    boundBone {boneName=dyn_shirtruffle_bt_r_jnt}
    componentAppliedLinearAccClamp=<6000 6000 6000>
    componentLinearAccScale=<0.4 0.4 0.4>
    componentLinearVelScale=<0.4 0.4 0.4>
    physicsBodyDefinitions [{
      boundBone {boneName=dyn_shirtruffle_bt_r_jnt}
      boxExtents=<10 10 10>
      constraintSetup {angularLimitsMax=<3 10 3>, angularLimitsMin=<-3 -10 -3>, angularTarget=<1 0 0>}
      localJointOffset=<12 0 0>
     }]
   }
   n121 = Anim Dynamics - Bone: dyn_ArmFrill_r_upper_jnt!(ComponentPose=@n120, Alpha=0.75, GravityOverride:Vector="0,0,0", ExternalForce="0,0,0") -> (Pose) {
    angularDampingOverride=1
    angularSpringConstant=200
    bAngularSpring=true
    bOverrideAngularDamping=true
    bUseGravityOverride=true
    simulationSpace=RootRelative
    alphaBoolBlend {blendOption=Linear}
    boundBone {boneName=dyn_ArmFrill_r_upper_jnt}
    componentAppliedLinearAccClamp=<6000 6000 6000>
    componentLinearAccScale=<0.4 0.4 0.4>
    componentLinearVelScale=<0.4 0.4 0.4>
    physicsBodyDefinitions [{
      boundBone {boneName=dyn_ArmFrill_r_upper_jnt}
      boxExtents=<10 10 10>
      constraintSetup {angularLimitsMax=<2 20 10>, angularLimitsMin=<-2 -20 -10>, angularTarget=<1 0 0>}
      localJointOffset=<10 0 0>
     }]
   }
   n122 = Anim Dynamics - Bone: dyn_ArmFrill_r_lower_jnt!(ComponentPose=@n121, Alpha=0.75, GravityOverride:Vector="0,0,0", ExternalForce="0,0,0") -> (Pose) {
    angularDampingOverride=1
    angularSpringConstant=200
    bAngularSpring=true
    bOverrideAngularDamping=true
    bUseGravityOverride=true
    simulationSpace=RootRelative
    alphaBoolBlend {blendOption=Linear}
    boundBone {boneName=dyn_ArmFrill_r_lower_jnt}
    componentAppliedLinearAccClamp=<6000 6000 6000>
    componentLinearAccScale=<0.4 0.4 0.4>
    componentLinearVelScale=<0.4 0.4 0.4>
    physicsBodyDefinitions [{
      boundBone {boneName=dyn_ArmFrill_r_lower_jnt}
      boxExtents=<10 10 10>
      constraintSetup {angularLimitsMax=<2 20 10>, angularLimitsMin=<-2 -20 -10>, angularTarget=<1 0 0>}
      localJointOffset=<10 0 0>
     }]
   }
   n123 = Anim Dynamics - Bone: dyn_ArmFrill_r_inner_jnt!(ComponentPose=@n122, Alpha=0.75, GravityOverride:Vector="0,0,0", ExternalForce="0,0,0") -> (Pose) {
    angularDampingOverride=1
    angularSpringConstant=200
    bAngularSpring=true
    bOverrideAngularDamping=true
    bUseGravityOverride=true
    simulationSpace=RootRelative
    alphaBoolBlend {blendOption=Linear}
    boundBone {boneName=dyn_ArmFrill_r_inner_jnt}
    componentAppliedLinearAccClamp=<6000 6000 6000>
    componentLinearAccScale=<0.4 0.4 0.4>
    componentLinearVelScale=<0.4 0.4 0.4>
    physicsBodyDefinitions [{
      boundBone {boneName=dyn_ArmFrill_r_inner_jnt}
      boxExtents=<10 10 10>
      constraintSetup {angularLimitsMax=<2 20 10>, angularLimitsMin=<-2 -20 -10>, angularTarget=<1 0 0>}
      localJointOffset=<10 0 0>
     }]
   }
   n124 = Anim Dynamics - Bone: dyn_ArmFrill_r_outter_jnt!(ComponentPose=@n123, Alpha=0.75, GravityOverride:Vector="0,0,0", ExternalForce="0,0,0") -> (Pose) {
    angularDampingOverride=1
    angularSpringConstant=200
    bAngularSpring=true
    bOverrideAngularDamping=true
    bUseGravityOverride=true
    simulationSpace=RootRelative
    alphaBoolBlend {blendOption=Linear}
    boundBone {boneName=dyn_ArmFrill_r_outter_jnt}
    componentAppliedLinearAccClamp=<6000 6000 6000>
    componentLinearAccScale=<0.4 0.4 0.4>
    componentLinearVelScale=<0.4 0.4 0.4>
    physicsBodyDefinitions [{
      boundBone {boneName=dyn_ArmFrill_r_outter_jnt}
      boxExtents=<10 10 10>
      constraintSetup {angularLimitsMax=<2 20 10>, angularLimitsMin=<-2 -20 -10>, angularTarget=<1 0 0>}
      localJointOffset=<10 0 0>
     }]
   }
   Save cached pose 'PreSim Cache'(Pose=Component To Local(ComponentPose=@n124)) {cacheName=PreSim Cache}
  ]}
 EventGraph {nodes [
   Event BlueprintUpdateAnimation()
   > n1 = Sequence!(execute) -> (then_0, then_1, then_2:Exec, then_3:Exec, then_4:Exec)
   > Set simSpace(simSpace:AnimPhysSimSpaceType Enum) -> (Output_Get:AnimPhysSimSpaceType Enum)
   > Branch(Condition=OR Boolean(A=Get bIsDBNO(), B=Get bIsInFrontEnd()))
   > n2 = Branch(Condition=OR Boolean(A=Get bIsDBNOCarried(), B=Get bIsInFrontEnd()))
   > Set simSpace(simSpace=World)
   Set simSpace(execute=@n2.else)
   n3 = PlayerMovement(execute=@n1.then_3, AnimInstance=Self-Reference(), XDotProductCoefficient=25, YDotProductCoefficient=25, ZDotProductCoefficient=25)
   > Set MovementAlpha(MovementAlpha:double="Absolute (Float)"(A=float * float(A=@n3.YMovement, B=0.045) {is=title K2Node_PromotableOperator})) -> (Output_Get:double)
   n4 = Sequence!(execute=@n1.then_4) -> (then_0, then_1, then_2, then_3, then_4, then_5:Exec, then_6:Exec, then_7:Exec)
   n5 = NOT Boolean(A=Get bIsMoving2D())
   n6 = Get bIsCrouching()
   > Set Is Crouching Not Moving Alpha(Is Crouching Not Moving Alpha:double="To Float (Boolean)"(InBool=AND Boolean(A=@n6, B=@n5))) -> (Output_Get:double)
   Set Is Knee Sliding Alpha(execute=@n4.then_1, Is Knee Sliding Alpha:double="To Float (Boolean)"(InBool=Get bIsKneeSliding())) -> (Output_Get:double)
   > Set Is Baseball Sliding(Is Baseball Sliding:double="To Float (Boolean)"(InBool=Get bIsBaseballSliding())) -> (Output_Get:double)
   > Set Is Sliding Alpha(Is Sliding Alpha:double="To Float (Boolean)"(InBool=Get bIsSliding())) -> (Output_Get:double)
   Set Is In Vehicle Alpha(execute=@n4.then_2, Is In Vehicle Alpha:double="To Float (Boolean)"(InBool=Get bIsInVehicle() -> (bIsInVehicle:bool))) -> (Output_Get:double)
   > Set Is On GrindRail Alpha(Is On GrindRail Alpha:double="To Float (Boolean)"(InBool=Get bIsOnGrindRail())) -> (Output_Get:double)
   > Set Is in FrontEnd Float(Is in FrontEnd Float:double="To Float (Boolean)"(InBool=Get bIsInFrontEnd())) -> (Output_Get:double)
   Set Is Emoting Alpha(execute=@n4.then_3, Is Emoting Alpha:double="To Float (Boolean)"(InBool=Get bIsPlayingEmote())) -> (Output_Get:double)
   n7 = Get bIsPlayingEmote()
   > Set Is not Emoting Alpha(Is not Emoting Alpha:double="To Float (Boolean)"(InBool=NOT Boolean(A=@n7))) -> (Output_Get:double)
   n8 = Sequence!(execute=@n4.then_4) -> (then_0, then_1, then_2, then_3, then_4)
   > n9 = Cast To FortPlayerAnimInstance(Object=GetAnimInstance(self=GetOwnerSkeletalMeshComponent()))
   > Set isPistol(isPistol=false)
   > Set isRifle(isRifle:bool=false) -> (Output_Get:bool)
   > n10 = Switch on EFortWeaponCoreAnimation!(execute, Selection:EFortWeaponCoreAnimation Enum=Get WeaponCoreAnim(self=@n9)) -> (
    Melee:Exec
    Pistol:Exec
    Shotgun:Exec
    PaperBlueprint:Exec
    Rifle:Exec
    MeleeOneHand:Exec
    MachinePistol:Exec
    RocketLauncher:Exec
    GrenadeLauncher:Exec
    GoingCommando:Exec
    AssaultRifle:Exec
    TacticalShotgun:Exec
    SniperRifle:Exec
    TrapPlacement:Exec
    ShoulderLauncher:Exec
    AbilityDecoTool:Exec
    Crossbow:Exec
    C4:Exec
    RemoteControl:Exec
    DualWield:Exec
    AR_BullPup:Exec
    AR_ForwardGrip:Exec
    MedPackPaddles:Exec
    SMG_P90:Exec
    AR_DrumGun:Exec
    Consumable_Small:Exec
    Consumable_Large:Exec
    Balloon:Exec
    MountedTurret:Exec
    CreativeTool:Exec
    ExplosiveBow:Exec
    AshtonIndigo:Exec
    AshtonChicago:Exec
    MeleeDualWield:Exec
    Unarmed:Exec
   )
   Set isRifle(execute=[@n10.AssaultRifle, @n10.Rifle, @n10.RocketLauncher, @n10.Shotgun, @n10.SniperRifle, @n10.TacticalShotgun], isRifle=true)
   Set isPistol(execute=@n10.Pistol, isPistol=true)
   Set Is Rifle Alpha(execute=@n8.then_1, Is Rifle Alpha:float="To Float (Boolean)"(InBool=Get isRifle() -> (isRifle:bool))) -> (Output_Get:float)
   > Set Is Pistol Alpha(Is Pistol Alpha:float="To Float (Boolean)"(InBool=Get isPistol() -> (isPistol:bool))) -> (Output_Get:float)
   n11 = Cloth Controller Function DEPRECATED!(
    execute=@n8.then_2
    CharacterPartAnimInstance:Custom Character Part Anim Instance Object Reference=Self-Reference()
    ClothingAssetName:String=M_MED_FlowerVase_Clothing_2
    Cloth Paramaters:Cloth Params Struct Structure=Get Cloth Paramaters_0() -> (Cloth Paramaters_0:Cloth Params Struct Structure)
    Cloth Core Settings:Cloth Core Setting Struct Structure=Get Cloth Core Settings_0() -> (Cloth Core Settings_0:Cloth Core Setting Struct Structure)
    ClothUpAxis:Vector="1,0,0"
    OnGround_Moving_SpeedLimit:double=2
    OnGround_Moving_AngularVel_Max:double=0.05
    OnGround_Moving_FictitiousVel_Max:double=0.6
    LinVelScale_FwdBwd_Clamp:double=0
    LinVelScale_UpDwn_Clamp:double=0.7
    DeltaTime:double=Get DeltaTimeLocal() -> (DeltaTimeLocal:double)
    GameTimeInSeconds:double=GetGameTimeInSeconds()
    WindGustsParams:Wind Gust Params Struct Structure=Get Wind Gust Params() -> (Wind Gust Params:Wind Gust Params Struct Structure)
    WindGustLocalSettings:Wind Gust Local Setting Struct Structure=Get Gust Local Settings() -> (Gust Local Settings:Wind Gust Local Setting Struct Structure)
   ) -> (
    then
    CurrentStateOut:String
    Cloth Settings Out:Cloth Setting Struct Structure
    Cloth Core Settings Out:Cloth Core Setting Struct Structure
    WindGustSettings Out:Wind Gust Setting Struct Structure
    WindGustLocalSettings Out:Wind Gust Local Setting Struct Structure
    GustStrength - WindVelocityOut:Vector
    LinearVelVector-DELETE LATER:Vector
   )
   > Set Gust Local Settings(Gust Local Settings:Wind Gust Local Setting Struct Structure=@n11.WindGustLocalSettings Out) -> (Output_Get:Wind Gust Local Setting Struct Structure)
   n12 = BreakRotator(InRot=Get PawnRotationRate() -> (PawnRotationRate:Rotator))
   > Branch(Condition="float > float"(A="Absolute (Float)"(A=@n12.Yaw), B=3) {is=title K2Node_CallFunction})
   > SetClothingInteractorIfNotSet()
   > SetVelocityScale!(
    execute
    self:Chaos Clothing Interactor Object Reference=Get ChaosClothingInteractor() -> (ChaosClothingInteractor:Chaos Clothing Interactor Object Reference)
    LinearVelocityScale:Vector="0.05,0.05,0.05"
    AngularVelocityScale:float=0
    FictitiousAngularScale:float=0
   ) -> (then)
   Set frontendAlpha(execute=@n1.then_1, frontendAlpha:double=SelectFloat(A=0.25, B=1, bPickA=Get bIsInFrontEnd())) -> (Output_Get:double)
   n13 = Gravity Override State Function DEPRECATED(
    execute=@n1.then_2
    CharacterPartAnimInstance=Self-Reference()
    GameTimeInSeconds=GetGameTimeInSeconds()
    GravityOverrideParams=Get Gravity Override Params()
   )
   > Set Gravity Override(Gravity Override=@n13)
   Event OnEndFalling()
   > SetClothingInteractorIfNotSet()
   > Branch(Condition=Get resetClothLandingFromZilpine() -> (resetClothLandingFromZilpine:bool))
   > Set SkipCCF(SkipCCF:bool=false) -> (Output_Get:bool)
   n14 = Get ChaosClothingInteractor()
   > ResetAndTeleport!(execute, self:Chaos Clothing Interactor Object Reference=@n14, bReset:bool=true, bTeleport:bool=true) -> (then)
   > Set resetClothLandingFromZilpine(resetClothLandingFromZilpine:bool=false) -> (Output_Get:bool)
   n15 = Event BlueprintProcessAnimInputEvent()
   > Branch(Condition=Get bClothEnabled())
   > SetClothingInteractorIfNotSet()
   > n16 = Branch(Condition=IsNamed(self=@n15.AnimInputEvent, InEventName=aie.Player.Zipline.Entered))
   > Set IsOnZipline(IsOnZipline:bool=true) -> (Output_Get:bool)
   > Delay(Duration=0.2)
   > ResetAndTeleport(self=@n14, bReset=true, bTeleport=true)
   > Set SkipCCF(SkipCCF=true)
   Set IsOnZipline(execute=@n16.else, IsOnZipline=false)
   > Branch(Condition=IsNamed(self=@n15.AnimInputEvent, InEventName=aie.Player.Zipline.Exited))
   > Set resetClothLandingFromZilpine(resetClothLandingFromZilpine=true)
   > Delay(Duration=1.1)
   > Set SkipCCF(SkipCCF=false)
  ]}
 SetClothingInteractorIfNotSet {nodes [SetClothingInteractorIfNotSet() {is=title K2Node_FunctionEntry}]}
]
variables [
 ChaosClothingInteractor {type=ChaosClothingInteractor, default=null}
 Cloth Core Settings_0 {
  type=ClothCoreSettingStruct
  default {
   collisionThickness=0
   dampingCoefficient=0.1
   frictionCoefficient=0.3
   localDampingCoefficient=0.1
   numberIterations=2
   numberSubsteps=1
   selfCollisionThickness=0
   animDriveDamping=<0 1>
   animDriveStiffness=<0 1>
   areaStiffness=<1 1>
   bendingStiffness=<1 1>
   edgeStiffness=<1 1>
   tetherStiffness=<1 1>
  }
 }
 Cloth Paramaters_0 {
  type=ClothParamsStruct
  default {
   dBNO {
    angularVelocityScale=0.6
    fictitiousAngularScale=0.1
    jointName=pelvis
    dragCoefficient=<0.2 1>
    gravityOverride=<-980 400 0>
    liftCoefficient=<0 0.5>
    linearVelocityScale=<0.3 0.3 0.1>
   }
   dBNO_Carried {
    angularVelocityScale=0.6
    fictitiousAngularScale=0.1
    jointName=pelvis
    dragCoefficient=<0.2 1>
    gravityOverride=<-980 -200 0>
    liftCoefficient=<0 0.5>
    linearVelocityScale=<0.3 0.3 0.1>
   }
   dBNO_Impulse {
    angularVelocityScale=0
    fictitiousAngularScale=0
    jointName=pelvis
    dragCoefficient=<0 0.2>
    gravityOverride=<-3000 -2000 0>
    liftCoefficient=<0 0.1>
    linearVelocityScale=<0.3 0.1 0.3>
   }
   dBNO_RollBwd {
    angularVelocityScale=0
    fictitiousAngularScale=0
    jointName=pelvis
    dragCoefficient=<0 0.2>
    gravityOverride=<-2000 0 0>
    liftCoefficient=<0 0.1>
    linearVelocityScale=<0.3 0.1 0.3>
   }
   dBNO_RollFwd {
    angularVelocityScale=0
    fictitiousAngularScale=0
    jointName=pelvis
    dragCoefficient=<0 0.2>
    gravityOverride=<-980 0 0>
    liftCoefficient=<0 0.1>
    linearVelocityScale=<0.3 0.1 0.3>
   }
   dBNO_Tumbling {
    angularVelocityScale=0
    fictitiousAngularScale=0
    jointName=pelvis
    dragCoefficient=<0 0.2>
    gravityOverride=<-2500 -300 0>
    liftCoefficient=<0 0.1>
    linearVelocityScale=<0.3 0.1 0.3>
   }
   emoteOrMelee {
    angularVelocityScale=0.2
    fictitiousAngularScale=0.1
    jointName=pelvis
    dragCoefficient=<0 0.3>
    gravityOverride=<-980 -500 0>
    liftCoefficient=<0 0.2>
    linearVelocityScale=<0.05 0.05 0.05>
   }
   front_End {
    angularVelocityScale=0.05
    fictitiousAngularScale=0.01
    jointName=root
    dragCoefficient=<0 2>
    gravityOverride=<0 -100 -980>
    liftCoefficient=<0 0.01>
    linearVelocityScale=<0.1 0.1 0.1>
   }
   onGround_Crouch_Moving {
    angularVelocityScale=0.1
    fictitiousAngularScale=0.3
    jointName=root
    dragCoefficient=<2 2>
    gravityOverride=<0 0 -980>
    liftCoefficient=<0.6 0.6>
    linearVelocityScale=<0.6 0.6 0.5>
   }
   onGround_Falling {
    angularVelocityScale=0.6
    fictitiousAngularScale=0.1
    jointName=root
    dragCoefficient=<0 0.2>
    gravityOverride=<0 -500 -980>
    liftCoefficient=<0 0.1>
    linearVelocityScale=<0.2 0.2 0.3>
   }
   onGround_Moving {
    angularVelocityScale=0.6
    fictitiousAngularScale=0.1
    jointName=root
    dragCoefficient=<0 0.3>
    gravityOverride=<0 -100 -980>
    liftCoefficient=<0 0.2>
    linearVelocityScale=<0.4 0.6 0.3>
   }
   onGround_Standing {
    angularVelocityScale=0.05
    fictitiousAngularScale=0.6
    jointName=root
    dragCoefficient=<0 2>
    gravityOverride=<0 -100 -980>
    liftCoefficient=<0 0.5>
    linearVelocityScale=<0.4 1 0.3>
   }
   rail_Grinding {
    angularVelocityScale=0.2
    fictitiousAngularScale=0
    jointName=root
    dragCoefficient=<0.1 0.2>
    gravityOverride=<50 -500 -400>
    liftCoefficient=<0 0.5>
    linearVelocityScale=<0.1 0.1 0.1>
   }
   riding_Moving {
    angularVelocityScale=0.05
    fictitiousAngularScale=1
    jointName=root
    dragCoefficient=<0.5 0.5>
    gravityOverride=<0 0 -980>
    liftCoefficient=<0.5 0.5>
    linearVelocityScale=<0.4 0.6 0.3>
   }
   riding_Moving_Targeting {
    angularVelocityScale=0.05
    fictitiousAngularScale=0.7
    jointName=root
    dragCoefficient=<0.5 0.5>
    gravityOverride=<0 0 -980>
    liftCoefficient=<0.5 0.5>
    linearVelocityScale=<0.2 0.6 0.2>
   }
   riding_Standing {
    angularVelocityScale=0.05
    fictitiousAngularScale=1
    jointName=root
    dragCoefficient=<0.5 0.5>
    gravityOverride=<0 -450 -980>
    liftCoefficient=<0.5 0.5>
    linearVelocityScale=<0.6 0.6 0.3>
   }
   skydive_DivingDown {
    angularVelocityScale=0.1
    fictitiousAngularScale=0
    jointName=pelvis
    dragCoefficient=<0 1.5>
    gravityOverride=<-450 -100 0>
    liftCoefficient=<0 0.3>
    linearVelocityScale=<0.1 0.1 0.1>
   }
   skydive_FloatingDown {
    angularVelocityScale=0.1
    fictitiousAngularScale=0
    jointName=root
    dragCoefficient=<0 2.5>
    gravityOverride=<0 0 980>
    liftCoefficient=<0 0.4>
    linearVelocityScale=<0.2 0.2 0.1>
   }
   skydive_GoingUp {
    angularVelocityScale=0.1
    fictitiousAngularScale=0
    jointName=root
    dragCoefficient=<0 0>
    gravityOverride=<0 -3000 3000>
    liftCoefficient=<0 0>
    linearVelocityScale=<0.1 0.1 0.05>
   }
   skydive_Parachuting {
    angularVelocityScale=0.1
    fictitiousAngularScale=0
    jointName=root
    dragCoefficient=<0 2.5>
    gravityOverride=<0 0 -980>
    liftCoefficient=<0 0.5>
    linearVelocityScale=<0.2 0.2 0.2>
   }
   skydive_Surfing {
    angularVelocityScale=0.6
    fictitiousAngularScale=0.1
    jointName=root
    dragCoefficient=<1.5 1.5>
    gravityOverride=<0 -600 -500>
    liftCoefficient=<1 1>
    linearVelocityScale=<0.1 0.1 0.3>
   }
   swimming {
    angularVelocityScale=0.6
    fictitiousAngularScale=0
    jointName=root
    dragCoefficient=<0 2>
    gravityOverride=<0 -600 400>
    liftCoefficient=<0 0.4>
    linearVelocityScale=<0.6 0.2 0.2>
   }
   tactical_Sprint {
    angularVelocityScale=0.6
    fictitiousAngularScale=0.1
    jointName=root
    dragCoefficient=<2 2>
    gravityOverride=<0 0 -980>
    liftCoefficient=<0.5 0.5>
    linearVelocityScale=<0.6 0.4 0.3>
   }
   ziplining {
    angularVelocityScale=0.6
    fictitiousAngularScale=0
    jointName=root
    dragCoefficient=<0 1.5>
    gravityOverride=<0 0 -980>
    liftCoefficient=<0 0.3>
    linearVelocityScale=<0.4 0.4 0.5>
   }
  }
 }
 DeltaTimeLocal {type=number}
 frontendAlpha {type=number}
 Gravity Override {type=Vector}
 Gravity Override Params {type=GravityOverrideParamsStruct}
 Gust Local Settings {type=WindGustLocalSettingStruct, default {elapsedTime=0, finalGustStrengthLocal=0, gustDuration=0, gustStrength=0, timeBetweenGusts=0}}
 Is Baseball Sliding {type=number}
 Is Crouching Not Moving Alpha {type=number}
 Is Emoting Alpha {type=number}
 Is in FrontEnd Float {type=number}
 Is In Vehicle Alpha {type=number}
 Is Knee Sliding Alpha {type=number}
 Is not Emoting Alpha {type=number}
 Is On GrindRail Alpha {type=number}
 Is Pistol Alpha {type=number}
 Is Rifle Alpha {type=number}
 Is Sliding Alpha {type=number}
 IsOnZipline {type=boolean}
 isPistol {type=boolean}
 isRifle {type=boolean}
 MovementAlpha {type=number}
 resetClothLandingFromZilpine {type=boolean}
 simSpace {type=AnimPhysSimSpaceType}
 SkipCCF {type=boolean}
 Wind Gust Params {
  type=WindGustParamsStruct
  default {
   dBNO {
    maxDownInterpSpeed=5
    maxGustDuration=2
    maxGustStrength=7
    maxTimeBetweenGusts=7
    maxUpInterpSpeed=5
    minDownInterpSpeed=2
    minGustDuration=0.5
    minGustStrength=3
    minTimeBetweenGusts=0
    minUpInterpSpeed=2
    windTurbulenceScalarA=0.1
    windTurbulenceScalarB=1
    windAmplitude=<0 0 0>
    windFrequency=<0 0 0>
    windOffset=<0 0 0>
   }
   dBNO_Carried {
    maxDownInterpSpeed=5
    maxGustDuration=2
    maxGustStrength=7
    maxTimeBetweenGusts=7
    maxUpInterpSpeed=5
    minDownInterpSpeed=2
    minGustDuration=0.5
    minGustStrength=3
    minTimeBetweenGusts=0
    minUpInterpSpeed=2
    windTurbulenceScalarA=0.1
    windTurbulenceScalarB=1
    windAmplitude=<0 0 0>
    windFrequency=<0 0 0>
    windOffset=<0 0 0>
   }
   dBNO_Impulse {
    maxDownInterpSpeed=5
    maxGustDuration=2
    maxGustStrength=7
    maxTimeBetweenGusts=7
    maxUpInterpSpeed=5
    minDownInterpSpeed=2
    minGustDuration=0.5
    minGustStrength=3
    minTimeBetweenGusts=0
    minUpInterpSpeed=2
    windTurbulenceScalarA=0.1
    windTurbulenceScalarB=1
    windAmplitude=<0 0 0>
    windFrequency=<0 0 0>
    windOffset=<0 0 0>
   }
   dBNO_RollBwd {
    maxDownInterpSpeed=5
    maxGustDuration=2
    maxGustStrength=7
    maxTimeBetweenGusts=7
    maxUpInterpSpeed=5
    minDownInterpSpeed=2
    minGustDuration=0.5
    minGustStrength=3
    minTimeBetweenGusts=0
    minUpInterpSpeed=2
    windTurbulenceScalarA=0.1
    windTurbulenceScalarB=1
    windAmplitude=<0 0 0>
    windFrequency=<0 0 0>
    windOffset=<0 0 0>
   }
   dBNO_RollFwd {
    maxDownInterpSpeed=5
    maxGustDuration=2
    maxGustStrength=7
    maxTimeBetweenGusts=7
    maxUpInterpSpeed=5
    minDownInterpSpeed=2
    minGustDuration=0.5
    minGustStrength=3
    minTimeBetweenGusts=0
    minUpInterpSpeed=2
    windTurbulenceScalarA=0.1
    windTurbulenceScalarB=1
    windAmplitude=<0 0 0>
    windFrequency=<0 0 0>
    windOffset=<0 0 0>
   }
   dBNO_Tumbling {
    maxDownInterpSpeed=5
    maxGustDuration=2
    maxGustStrength=7
    maxTimeBetweenGusts=7
    maxUpInterpSpeed=5
    minDownInterpSpeed=2
    minGustDuration=0.5
    minGustStrength=3
    minTimeBetweenGusts=0
    minUpInterpSpeed=2
    windTurbulenceScalarA=0.1
    windTurbulenceScalarB=1
    windAmplitude=<0 0 0>
    windFrequency=<0 0 0>
    windOffset=<0 0 0>
   }
   emoteOrMelee {
    maxDownInterpSpeed=5
    maxGustDuration=2
    maxGustStrength=7
    maxTimeBetweenGusts=7
    maxUpInterpSpeed=5
    minDownInterpSpeed=2
    minGustDuration=0.5
    minGustStrength=3
    minTimeBetweenGusts=0
    minUpInterpSpeed=2
    windTurbulenceScalarA=0.1
    windTurbulenceScalarB=1
    windAmplitude=<30 25 50>
    windFrequency=<2 3 5>
    windOffset=<-300 -150 0>
   }
   front_End {
    maxDownInterpSpeed=5
    maxGustDuration=2
    maxGustStrength=4
    maxTimeBetweenGusts=7
    maxUpInterpSpeed=5
    minDownInterpSpeed=2
    minGustDuration=0.5
    minGustStrength=2
    minTimeBetweenGusts=0
    minUpInterpSpeed=2
    windTurbulenceScalarA=0.1
    windTurbulenceScalarB=1
    windAmplitude=<10 30 50>
    windFrequency=<10 3 4>
    windOffset=<0 -200 -250>
   }
   onGround_Crouch_Moving {
    maxDownInterpSpeed=5
    maxGustDuration=2
    maxGustStrength=7
    maxTimeBetweenGusts=7
    maxUpInterpSpeed=5
    minDownInterpSpeed=2
    minGustDuration=0.5
    minGustStrength=3
    minTimeBetweenGusts=0
    minUpInterpSpeed=2
    windTurbulenceScalarA=0.1
    windTurbulenceScalarB=1
    windAmplitude=<100 25 75>
    windFrequency=<1.5 0.5 7.5>
    windOffset=<0 -100 -175>
   }
   onGround_Falling {
    maxDownInterpSpeed=5
    maxGustDuration=2
    maxGustStrength=7
    maxTimeBetweenGusts=7
    maxUpInterpSpeed=5
    minDownInterpSpeed=2
    minGustDuration=0.5
    minGustStrength=3
    minTimeBetweenGusts=0
    minUpInterpSpeed=2
    windTurbulenceScalarA=0.1
    windTurbulenceScalarB=1
    windAmplitude=<0 0 0>
    windFrequency=<0 0 0>
    windOffset=<0 0 0>
   }
   onGround_Moving {
    maxDownInterpSpeed=5
    maxGustDuration=8
    maxGustStrength=7
    maxTimeBetweenGusts=8
    maxUpInterpSpeed=5
    minDownInterpSpeed=2
    minGustDuration=0.2
    minGustStrength=3
    minTimeBetweenGusts=1
    minUpInterpSpeed=2
    windTurbulenceScalarA=0.1
    windTurbulenceScalarB=0.8
    windAmplitude=<30 65 42>
    windFrequency=<5 6 8.6>
    windOffset=<30 -250 -300>
   }
   onGround_Standing {
    maxDownInterpSpeed=5
    maxGustDuration=8
    maxGustStrength=7
    maxTimeBetweenGusts=8
    maxUpInterpSpeed=5
    minDownInterpSpeed=2
    minGustDuration=0.2
    minGustStrength=2
    minTimeBetweenGusts=1
    minUpInterpSpeed=2
    windTurbulenceScalarA=0.1
    windTurbulenceScalarB=0.8
    windAmplitude=<75 24 28>
    windFrequency=<1.5 2.3 3.7>
    windOffset=<75 -100 -150>
   }
   rail_Grinding {
    maxDownInterpSpeed=5
    maxGustDuration=8
    maxGustStrength=7
    maxTimeBetweenGusts=8
    maxUpInterpSpeed=5
    minDownInterpSpeed=2
    minGustDuration=0.2
    minGustStrength=3
    minTimeBetweenGusts=1
    minUpInterpSpeed=2
    windTurbulenceScalarA=0.1
    windTurbulenceScalarB=0.8
    windAmplitude=<20 65 42>
    windFrequency=<5 6 8.6>
    windOffset=<50 -500 -400>
   }
   riding_Moving {
    maxDownInterpSpeed=5
    maxGustDuration=2
    maxGustStrength=7
    maxTimeBetweenGusts=7
    maxUpInterpSpeed=5
    minDownInterpSpeed=2
    minGustDuration=0.5
    minGustStrength=3
    minTimeBetweenGusts=0
    minUpInterpSpeed=2
    windTurbulenceScalarA=0.1
    windTurbulenceScalarB=1
    windAmplitude=<0 0 0>
    windFrequency=<0 0 0>
    windOffset=<0 0 0>
   }
   riding_Moving_Targeting {
    maxDownInterpSpeed=5
    maxGustDuration=2
    maxGustStrength=7
    maxTimeBetweenGusts=7
    maxUpInterpSpeed=5
    minDownInterpSpeed=2
    minGustDuration=0.5
    minGustStrength=3
    minTimeBetweenGusts=0
    minUpInterpSpeed=2
    windTurbulenceScalarA=0.1
    windTurbulenceScalarB=1
    windAmplitude=<0 0 0>
    windFrequency=<0 0 0>
    windOffset=<0 0 0>
   }
   riding_Standing {
    maxDownInterpSpeed=5
    maxGustDuration=2
    maxGustStrength=7
    maxTimeBetweenGusts=7
    maxUpInterpSpeed=5
    minDownInterpSpeed=2
    minGustDuration=0.5
    minGustStrength=3
    minTimeBetweenGusts=0
    minUpInterpSpeed=2
    windTurbulenceScalarA=0.1
    windTurbulenceScalarB=1
    windAmplitude=<0 0 0>
    windFrequency=<0 0 0>
    windOffset=<0 0 0>
   }
   skydive_DivingDown {
    maxDownInterpSpeed=5
    maxGustDuration=8
    maxGustStrength=7
    maxTimeBetweenGusts=8
    maxUpInterpSpeed=5
    minDownInterpSpeed=2
    minGustDuration=0.2
    minGustStrength=3
    minTimeBetweenGusts=1
    minUpInterpSpeed=2
    windTurbulenceScalarA=0.1
    windTurbulenceScalarB=4
    windAmplitude=<0 25 35>
    windFrequency=<0 4 10>
    windOffset=<-450 -100 100>
   }
   skydive_FloatingDown {
    maxDownInterpSpeed=5
    maxGustDuration=8
    maxGustStrength=1
    maxTimeBetweenGusts=8
    maxUpInterpSpeed=5
    minDownInterpSpeed=2
    minGustDuration=0.2
    minGustStrength=0
    minTimeBetweenGusts=1
    minUpInterpSpeed=2
    windTurbulenceScalarA=0.5
    windTurbulenceScalarB=1
    windAmplitude=<50 20 35>
    windFrequency=<8 2 3>
    windOffset=<0 -300 200>
   }
   skydive_GoingUp {
    maxDownInterpSpeed=5
    maxGustDuration=8
    maxGustStrength=7
    maxTimeBetweenGusts=8
    maxUpInterpSpeed=5
    minDownInterpSpeed=2
    minGustDuration=0.2
    minGustStrength=3
    minTimeBetweenGusts=1
    minUpInterpSpeed=2
    windTurbulenceScalarA=0.1
    windTurbulenceScalarB=0.8
    windAmplitude=<0 0 0>
    windFrequency=<0 0 0>
    windOffset=<0 -500 300>
   }
   skydive_Parachuting {
    maxDownInterpSpeed=5
    maxGustDuration=8
    maxGustStrength=1
    maxTimeBetweenGusts=8
    maxUpInterpSpeed=5
    minDownInterpSpeed=2
    minGustDuration=5
    minGustStrength=0
    minTimeBetweenGusts=5
    minUpInterpSpeed=2
    windTurbulenceScalarA=0.2
    windTurbulenceScalarB=1
    windAmplitude=<25 50 75>
    windFrequency=<8 6 2>
    windOffset=<-150 -400 -150>
   }
   skydive_Surfing {
    maxDownInterpSpeed=5
    maxGustDuration=2
    maxGustStrength=7
    maxTimeBetweenGusts=7
    maxUpInterpSpeed=5
    minDownInterpSpeed=2
    minGustDuration=0.5
    minGustStrength=3
    minTimeBetweenGusts=0
    minUpInterpSpeed=2
    windTurbulenceScalarA=0.1
    windTurbulenceScalarB=1
    windAmplitude=<0 0 0>
    windFrequency=<0 0 0>
    windOffset=<0 0 0>
   }
   swimming {
    maxDownInterpSpeed=5
    maxGustDuration=2
    maxGustStrength=7
    maxTimeBetweenGusts=7
    maxUpInterpSpeed=5
    minDownInterpSpeed=2
    minGustDuration=0.5
    minGustStrength=3
    minTimeBetweenGusts=0
    minUpInterpSpeed=2
    windTurbulenceScalarA=0.1
    windTurbulenceScalarB=1
    windAmplitude=<10 25 20>
    windFrequency=<1.5 0.5 7.5>
    windOffset=<0 -150 50>
   }
   tactical_Sprint {
    maxDownInterpSpeed=5
    maxGustDuration=2
    maxGustStrength=7
    maxTimeBetweenGusts=7
    maxUpInterpSpeed=5
    minDownInterpSpeed=2
    minGustDuration=0.5
    minGustStrength=3
    minTimeBetweenGusts=0
    minUpInterpSpeed=2
    windTurbulenceScalarA=0.1
    windTurbulenceScalarB=1
    windAmplitude=<0 0 0>
    windFrequency=<0 0 0>
    windOffset=<0 0 0>
   }
   ziplining {
    maxDownInterpSpeed=5
    maxGustDuration=2
    maxGustStrength=7
    maxTimeBetweenGusts=7
    maxUpInterpSpeed=5
    minDownInterpSpeed=2
    minGustDuration=0.5
    minGustStrength=3
    minTimeBetweenGusts=0
    minUpInterpSpeed=2
    windTurbulenceScalarA=0.1
    windTurbulenceScalarB=1
    windAmplitude=<0 0 0>
    windFrequency=<0 0 0>
    windOffset=<0 0 0>
   }
  }
 }
]
