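# Control Rig for the face accessory part of the male medium FlowerVase character, applying pose-driven corrective offsets to its hat and mustache bones.
# A spherical pose reader on neck_01, relative to spine_05, weights transform offsets on the hat joint and its back master bone.
# Two mirrored readers on the left and right upper lip bones, relative to the head, likewise offset the second joint of each mustache side.
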ControlRigBlueprint /BRCosmetics/Characters/Player/Male/Medium/Bodies/M_MED_FlowerVase/Meshes/Parts/M_MED_FlowerVase_FaceAcc_CtrlRig
graphs [
 RigVMFunctionLibrary {links [], nodes []}
 RigVMModel {nodes [
   RigUnit_BeginExecution = &RigUnit_BeginExecution()
   n1 = &RigUnit_SphericalPoseReader(
    ActiveRegionScaleFactors=(PositiveWidth=0.5,NegativeWidth=1,PositiveHeight=2,NegativeHeight=2)
    ActiveRegionSize=0.1
    Debug=(bDrawDebug=false,DebugScale=25,DebugSegments=20,DebugThickness=0.25)
    DriverAxis=(X=0,Y=-1,Z=0)
    DriverItem=(Type=Bone,Name="neck_01")
    FalloffRegionScaleFactors=(PositiveWidth=0.5,NegativeWidth=0.5,PositiveHeight=2,NegativeHeight=2)
    FalloffSize=0.3
    FlipWidthScaling=false
    OptionalParentItem=(Type=Bone,Name="spine_05")
    RotationOffset=(X=-90,Y=0,Z=-30)
    ExecutePin=@RigUnit_BeginExecution.ExecutePin
   )
   > &RigUnit_ModifyTransforms(
    ItemToModify=((Item=(Type=Bone,Name="dyn_Hat_bk_c_MSTR"),Transform=(Rotation=(X=0.0057,Y=0.0432,Z=-0.1304,W=0.9905),Translation=(X=1,Y=0,Z=0),Scale3D=(X=1,Y=1,Z=1))))
    WeightMinimum=0
    Weight=@n1.OutputParam
   )
   > &RigUnit_ModifyTransforms(
    ItemToModify=((Item=(Type=Bone,Name="Hat_jnt"),Transform=(Rotation=(X=0,Y=0,Z=0,W=1),Translation=(X=0.3,Y=0,Z=0),Scale3D=(X=1,Y=1,Z=1))))
    WeightMinimum=0
    Weight=@n1.OutputParam
   )
   > n2 = &RigUnit_SphericalPoseReader(
    ActiveRegionScaleFactors=(PositiveWidth=0.75,NegativeWidth=0.75,PositiveHeight=0.75,NegativeHeight=0.75)
    ActiveRegionSize=0.1
    Debug=(bDrawDebug=false,DebugScale=25,DebugSegments=20,DebugThickness=0.25)
    DriverAxis=(X=0,Y=0,Z=1)
    DriverItem=(Type=Bone,Name="FACIAL_L_LipUpper")
    FalloffRegionScaleFactors=(PositiveWidth=0.5,NegativeWidth=0.5,PositiveHeight=0.5,NegativeHeight=0.5)
    FalloffSize=0.3
    FlipWidthScaling=false
    OptionalParentItem=(Type=Bone,Name="head")
    RotationOffset=(X=-45,Y=0,Z=10)
   )
   > &RigUnit_ModifyTransforms(
    ItemToModify=((Item=(Type=Bone,Name="dyn_Mustache_l_02_jnt"),Transform=(Rotation=(X=0,Y=0,Z=0.1305,W=0.9914),Translation=(X=0,Y=1,Z=1),Scale3D=(X=1,Y=1,Z=1))))
    WeightMinimum=0
    Weight=@n2.OutputParam
   )
   > n3 = &RigUnit_SphericalPoseReader(
    ActiveRegionScaleFactors=(PositiveWidth=0.75,NegativeWidth=0.75,PositiveHeight=0.75,NegativeHeight=0.75)
    ActiveRegionSize=0.1
    Debug=(bDrawDebug=false,DebugScale=25,DebugSegments=20,DebugThickness=0.25)
    DriverAxis=(X=0,Y=0,Z=1)
    DriverItem=(Type=Bone,Name="FACIAL_R_LipUpper")
    FalloffRegionScaleFactors=(PositiveWidth=0.5,NegativeWidth=0.5,PositiveHeight=0.5,NegativeHeight=0.5)
    FalloffSize=0.3
    FlipWidthScaling=false
    OptionalParentItem=(Type=Bone,Name="head")
    RotationOffset=(X=-45,Y=0,Z=-10)
   )
   > &RigUnit_ModifyTransforms(
    ItemToModify=((Item=(Type=Bone,Name="dyn_Mustache_r_02_jnt"),Transform=(Rotation=(X=0,Y=0,Z=-0.1305,W=0.9914),Translation=(X=0,Y=-1,Z=1),Scale3D=(X=1,Y=1,Z=1))))
    WeightMinimum=0
    Weight=@n3.OutputParam
   )
  ]}
]
hierarchy [
 root {loc=<0 0 0>, rot=<0 0 0 1>}
 pelvis {parent=root, loc=<0 -0.7071 93.1191>, rot=<0.0127 0.707 -0.0127 -0.707>}
 spine_01 {parent=pelvis, loc=<6.7494 -0.6388 0>, rot=<0 0 0.034 -0.9994>}
 spine_02 {parent=spine_01, loc=<7.2518 -0.035 0>, rot=<0 0 0.0238 -0.9997>}
 spine_03 {parent=spine_02, loc=<6.0104 0 0>, rot=<0 0 0.0149 -0.9999>}
 spine_04 {parent=spine_03, loc=<8.1024 0 0>, rot=<0 0 -0.0132 -0.9999>}
 spine_05 {parent=spine_04, loc=<10.0146 0 0>, rot=<0 0 -0.0015 -1>}
 clavicle_l {parent=spine_05, loc=<4.3575 -0.1466 -2.3223>, rot=<0.0465 -0.7289 -0.009 -0.6829>}
 neck_01 {parent=spine_05, loc=<11.3138 0.1529 0>, rot=<0 0 -0.2003 -0.9797>}
 neck_02 {parent=neck_01, loc=<4.5627 0 0>, rot=<0 0 0.0504 -0.9987>}
 head {parent=neck_02, loc=<5.3914 0 0>, rot=<0 0 0.115 -0.9934>}
 dyn_Mustache_l_MSTR {parent=head, loc=<-0.6532 9.8504 -0.5573>, rot=<0.0029 -0.7071 -0.0029 -0.7071>}
 dyn_Mustache_l_01_jnt {parent=dyn_Mustache_l_MSTR, loc=<0 0 0>, rot=<-0.1205 -0.0743 0.2213 -0.9649>}
 dyn_Mustache_l_02_jnt {parent=dyn_Mustache_l_01_jnt, loc=<2.8223 0 0>, rot=<0.1725 0.1956 -0.1201 -0.9579>}
 dyn_Mustache_l_03_jnt {parent=dyn_Mustache_l_02_jnt, loc=<2.3509 0 0>, rot=<0.0234 0.2159 0.0052 -0.9761>}
 dyn_Mustache_r_MSTR {parent=head, loc=<-0.6537 9.8504 0.5573>, rot=<0.7071 0.0029 0.7071 -0.0029>}
 dyn_Mustache_r_01_jnt {parent=dyn_Mustache_r_MSTR, loc=<0 0 0>, rot=<0.1205 -0.0743 -0.2213 -0.9649>}
 dyn_Mustache_r_02_jnt {parent=dyn_Mustache_r_01_jnt, loc=<2.8223 0 0>, rot=<-0.1725 0.1956 0.1201 -0.9579>}
 dyn_Mustache_r_03_jnt {parent=dyn_Mustache_r_02_jnt, loc=<2.351 0.0001 0.0009>, rot=<-0.0234 0.2159 -0.0052 -0.9761>}
 Hat_jnt {parent=head, loc=<7.5208 1.421 0>, rot=<1 0.0041 0 0>}
 dyn_Hat_flap_l_MSTR {parent=Hat_jnt, loc=<0.7302 -1.3199 11.1632>, rot=<-0.2132 0 0 -0.977>}
 dyn_Hat_flap_l_01_jnt {parent=dyn_Hat_flap_l_MSTR, loc=<0 0 0>, rot=<0 0.0157 0 -0.9999>}
 dyn_Hat_flap_r_MSTR {parent=Hat_jnt, loc=<0.7305 -1.3199 -11.1632>, rot=<0.977 0 0 0.2132>}
 dyn_Hat_flap_r_01_jnt {parent=dyn_Hat_flap_r_MSTR, loc=<0 0 0>, rot=<0 0.0157 0 -0.9999>}
 dyn_Hat_Fluff_MSTR {parent=Hat_jnt, loc=<-1.0817 4.2928 -5.31>, rot=<0 0 0 1>}
 dyn_Hat_Fluff_01_jnt {parent=dyn_Hat_Fluff_MSTR, loc=<0 0 0>, rot=<0.0671 -0.0637 -0.3331 -0.9383>}
 dyn_Hat_Fluff_02_jnt {parent=dyn_Hat_Fluff_01_jnt, loc=<10.574 0 0>, rot=<0.0671 -0.0637 -0.3331 -0.9383>}
 dyn_Hat_bk_c_MSTR {parent=Hat_jnt, loc=<-3.0889 7.424 0>, rot=<0.7071 0 0 -0.7071>}
 dyn_Hat_bk_c_jnt {parent=dyn_Hat_bk_c_MSTR, loc=<0 0 0>, rot=<0 0.2216 0 -0.9751>}
 FACIAL_C_FacialRoot {parent=head, loc=<2.1524 -0.746 -0.0002>, rot=<0.4879 -0.5118 -0.4879 -0.5118>}
 FACIAL_L_NasolabialFurrow {parent=FACIAL_C_FacialRoot, loc=<2.7591 2.1015 9.341>, rot=<-0.0303 -0.127 -0.3632 -0.9225>}
 FACIAL_R_NasolabialFurrow {parent=FACIAL_C_FacialRoot, loc=<-2.7552 2.105 9.3372>, rot=<-0.0303 0.127 0.3632 -0.9225>}
 FACIAL_C_Nose {parent=FACIAL_C_FacialRoot, loc=<-0.0005 -0.3134 10.5844>, rot=<-0.1271 0 0 -0.9919>}
 FACIAL_C_NoseLower {parent=FACIAL_C_Nose, loc=<0.0001 2.9307 -0.6318>, rot=<0.1305 0 0 -0.9914>}
 FACIAL_L_Nostril {parent=FACIAL_C_Nose, loc=<1.2513 2.2712 -0.4781>, rot=<0.0173 0.1305 0.0023 -0.9913>}
 FACIAL_R_Nostril {parent=FACIAL_C_Nose, loc=<-1.2392 2.2458 -0.4803>, rot=<0.0173 -0.1305 -0.0023 -0.9913>}
 FACIAL_C_NoseTip {parent=FACIAL_C_Nose, loc=<-0.0032 2.6073 0.706>, rot=<0.0872 0 0 -0.9962>}
 FACIAL_C_MouthUpper {parent=FACIAL_C_FacialRoot, loc=<0.0003 4.2461 7.9419>, rot=<0.0033 0 0 -1>}
 FACIAL_C_LipUpper {parent=FACIAL_C_MouthUpper, loc=<-0.0015 -0.4529 2.0805>, rot=<-0.0872 0 0 -0.9962>}
 FACIAL_L_LipUpper {parent=FACIAL_C_MouthUpper, loc=<1.108 -0.3282 1.6921>, rot=<-0.0778 -0.1601 -0.0379 -0.9833>}
 FACIAL_R_LipUpper {parent=FACIAL_C_MouthUpper, loc=<-1.11 -0.3303 1.6931>, rot=<-0.0778 0.1601 0.0379 -0.9833>}
 FACIAL_L_LipUpperOuter {parent=FACIAL_C_MouthUpper, loc=<1.9932 -0.0882 1.1627>, rot=<-0.0984 -0.2293 -0.0983 -0.9634>}
 FACIAL_R_LipUpperOuter {parent=FACIAL_C_MouthUpper, loc=<-1.9973 -0.0932 1.16>, rot=<-0.0984 0.2293 0.0983 -0.9634>}
 FACIAL_L_LipCorner {parent=FACIAL_C_MouthUpper, loc=<2.5483 0.2004 0.7764>, rot=<0 -0.2078 0 -0.9782>}
 FACIAL_R_LipCorner {parent=FACIAL_C_MouthUpper, loc=<-2.537 0.1962 0.7761>, rot=<0 0.2078 0 -0.9782>}
 FACIAL_L_EyeCornerOuter {parent=FACIAL_C_FacialRoot, loc=<3.804 -1.493 8.5197>, rot=<0.003 -0.3826 0.0012 -0.9239>}
 FACIAL_L_EyesackLower {parent=FACIAL_C_FacialRoot, loc=<2.9796 -0.4743 8.6438>, rot=<-0.019 -0.1398 -0.0218 -0.9898>}
 FACIAL_R_EyesackLower {parent=FACIAL_C_FacialRoot, loc=<-2.9802 -0.4849 8.6351>, rot=<-0.019 0.1398 0.0218 -0.9898>}
 FACIAL_L_CheekInner {parent=FACIAL_C_FacialRoot, loc=<1.9347 0.2316 8.7021>, rot=<-0.12 -0.0887 -0.1968 -0.969>}
 FACIAL_R_CheekInner {parent=FACIAL_C_FacialRoot, loc=<-1.9398 0.2315 8.6998>, rot=<-0.1201 0.0886 0.1968 -0.969>}
 FACIAL_L_CheekOuter {parent=FACIAL_C_FacialRoot, loc=<3.6244 0.2969 7.255>, rot=<0.0582 -0.3766 0.1016 -0.9189>}
 FACIAL_R_CheekOuter {parent=FACIAL_C_FacialRoot, loc=<-3.6229 0.313 7.2502>, rot=<0.0582 0.3766 -0.1016 -0.919>}
 FACIAL_L_NasolabialBulge {parent=FACIAL_C_FacialRoot, loc=<3.5367 3.72 7.0559>, rot=<0.0609 -0.3773 0.025 -0.9238>}
 FACIAL_R_NasolabialBulge {parent=FACIAL_C_FacialRoot, loc=<-3.5375 3.7201 7.0509>, rot=<0.0609 0.3772 -0.0251 -0.9238>}
 FACIAL_L_CheekLower {parent=FACIAL_C_FacialRoot, loc=<4.0765 3.4787 6.4587>, rot=<0.0375 -0.4267 0.0591 -0.9017>}
 FACIAL_R_CheekLower {parent=FACIAL_C_FacialRoot, loc=<-4.0768 3.4783 6.453>, rot=<0.0375 0.4267 -0.0592 -0.9017>}
 FACIAL_L_Masseter {parent=FACIAL_C_FacialRoot, loc=<5.8109 1.0179 4.8094>, rot=<0.0726 0.1129 0.0075 -0.9909>}
 FACIAL_R_Masseter {parent=FACIAL_C_FacialRoot, loc=<-5.8124 1.02 4.8078>, rot=<0.0726 -0.1129 -0.0075 -0.9909>}
 FACIAL_C_LowerLipRotation {parent=FACIAL_C_FacialRoot, loc=<0.0023 1.8685 3.1375>, rot=<0.3451 0 0 -0.9386>}
 FACIAL_C_MouthLower {parent=FACIAL_C_LowerLipRotation, loc=<-0.0021 -1.3013 5.2002>, rot=<-0.342 0 0 -0.9397>}
 FACIAL_L_LipLower {parent=FACIAL_C_MouthLower, loc=<0.9335 0.7067 1.9514>, rot=<0.0927 -0.1294 0.0272 -0.9869>}
 FACIAL_R_LipLower {parent=FACIAL_C_MouthLower, loc=<-0.9414 0.7052 1.9493>, rot=<0.0927 0.1294 -0.0272 -0.9869>}
 FACIAL_L_LipLowerOuter {parent=FACIAL_C_MouthLower, loc=<1.9494 0.5267 1.3013>, rot=<0.0913 -0.2321 0.0318 -0.9679>}
 FACIAL_R_LipLowerOuter {parent=FACIAL_C_MouthLower, loc=<-1.9587 0.5231 1.2975>, rot=<0.0913 0.2321 -0.0318 -0.9679>}
 FACIAL_C_Jaw {parent=FACIAL_C_FacialRoot, loc=<0.0023 1.8685 3.1375>, rot=<0.3451 0 0 -0.9386>}
 FACIAL_L_ChinSide {parent=FACIAL_C_Jaw, loc=<2.9365 0.9252 7.5145>, rot=<-0.2208 -0.3738 0.0499 -0.8995>}
 FACIAL_R_ChinSide {parent=FACIAL_C_Jaw, loc=<-2.939 0.9301 7.5122>, rot=<-0.2208 0.3738 -0.0499 -0.8995>}
 FACIAL_L_Jawline {parent=FACIAL_C_FacialRoot, loc=<4.4129 5.3137 5.3483>, rot=<0.2125 0.1575 -0.0552 -0.9628>}
 FACIAL_R_Jawline {parent=FACIAL_C_FacialRoot, loc=<-4.4151 5.3126 5.3463>, rot=<0.2125 -0.1576 0.0552 -0.9628>}
 FACIAL_L_EyelidLowerB {parent=FACIAL_C_FacialRoot, loc=<2.539 -1.7947 8.2436>, rot=<0.3204 0 0 -0.9473>}
 dyn_Hair_l_MSTR {parent=spine_05, loc=<12.2593 8.2888 -8.285>, rot=<-0.0072 -0.8615 -0.504 -0.0617>}
 dyn_Hair_l_jnt {parent=dyn_Hair_l_MSTR, loc=<0 0 0>, rot=<0 0 0 1>}
 dyn_Hair_r_MSTR {parent=spine_05, loc=<12.2592 8.2888 8.285>, rot=<0.0617 0.504 0.8615 0.0072>}
 dyn_Hair_r_jnt {parent=dyn_Hair_r_MSTR, loc=<0 0 0>, rot=<0 0 0 -1>}
 clavicle_r {parent=spine_05, loc=<4.3575 -0.1466 2.3223>, rot=<0.7289 0.0465 -0.6829 0.009>}
]
